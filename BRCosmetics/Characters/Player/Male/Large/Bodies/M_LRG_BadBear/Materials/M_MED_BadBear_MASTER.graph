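# Master material for the BadBear cosmetic skin, with parameters separating character and glider use, adding animated glowing vines to the standard character look.
# Time-scrolled noise distorts the two-color vine masks, while timed pulse bands from pre-skinned position spheres boost their emissive.
# Sphere masks around two heart locations build a vertex-push socket offset.
# Character PreFX and PostFX functions wrap the effects, with static switches for Slurp and socket debugging.
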Material /BRCosmetics/Characters/Player/Male/Large/Bodies/M_LRG_BadBear/Materials/M_MED_BadBear_MASTER
graph {
 properties {
  bUseMaterialAttributes=true
  bUsedWithMeshDeformer=true
  bUsedWithMorphTargets=true
  bUsedWithNiagaraMeshParticles=true
  bUsedWithSkeletalMesh=true
  pixelDepthOffsetMode=PDOM_Legacy
  previewMesh=/BRCosmetics/Weapons/FORT_Melee/Pickaxe_BadBear/Meshes/BadBear_Axe
  usedShadingModels=Default Lit
 }
 nodes [
  BumpOffset(Height=Constant() {r=-0.1})
  Fresnel() {exponent=1}
  n1 = MaterialFunctionCall(In=Fresnel() {exponent=1}, Contrast=Constant() {r=0.5}) {materialFunction=&CheapContrast}
  LinearInterpolate(
   A=@n1.Result
   B=MaterialFunctionCall(In=Fresnel() {exponent=0.2}, Contrast=Constant() {r=0.001}) {materialFunction=&CheapContrast}.Result
   Alpha=SphereMask(A=NamedRerouteUsage(), B=Constant() {r=0.23}) {attenuationRadius=0.15, hardnessPercent=90}
  )
  n2 = TextureObject() {texture=/Game/CharClasses/Outlander/Gadgets/AntiMaterialCharge/Materials/T_AntiMaterialCharge_Dissolve_01}
  n3 = NamedRerouteUsage()
  n4 = Time()
  n5 = MaterialFunctionCall!(
   ColorTex=@n2
   NormalTex
   Pos=Add(A=Divide(A=@n3) {constB=0.5}, B=Multiply(A=@n4) {constB=8})
   Nrm
   ColorUVScale
   ColorBlendFactor=Constant() {r=20}
   NormalStrength
   NormalUVScale
   NormalBlendFactor
   Local Space Inputs
  ) {materialFunction=/Game/Global/MaterialFunctions/MF_BiPlanar}
  Multiply(
   A=ConstantBiasScale(None=ComponentMask(None=@n5.Color) {b=true}) {bias=-0.5, scale=2}
   B=LinearInterpolate(Alpha=NamedRerouteUsage()) {constB=0}
  ) {constB=0}
  n6 = NamedRerouteUsage()
  Multiply(A=@n6) {constB=0.1}
  n7 = NamedRerouteUsage()
  n8 = Time()
  n9 = MaterialFunctionCall(ColorTex=@n2, Pos=Add(A=Divide(A=@n7) {constB=0.75}, B=Multiply(A=@n8) {constB=10})) {materialFunction=&MF_BiPlanar}
  Multiply(
   A=ConstantBiasScale(None=ComponentMask(None=@n9.Color) {r=true}) {bias=-0.5, scale=2}
   B=ScalarParameter() {defaultValue=0.01, parameterName=BB_VineDistortionAmount2}
  ) {constB=0.01}
  n10 = Abs(None=NamedRerouteUsage())
  n11 = Constant3Vector() {constant=<rgba 70 10 100 0>}
  n12 = ComponentMask(None=@n10) {b=true}
  n13 = Constant()
  n14 = Power(Base=Add(
    A=SphereMask(A=@n10, B=@n11) {attenuationRadius=35, hardnessPercent=0}
    B=SphereMask(A=@n12, B=@n13) {attenuationRadius=40, hardnessPercent=0}
   )) {constExponent=1}
  n15 = ScalarParameter() {defaultValue=1, parameterName=IsCharacter}
  n16 = TextureSampleParameter2D() {group=Slurp, parameterName=Slurp_WPO_Mask, samplerType=SAMPLERTYPE_Masks, texture=&T_WhiteMask}
  NamedRerouteDeclaration(None=LinearInterpolate(A=LinearInterpolate(B=@n14, Alpha=@n15), B=OneMinus(None=@n16.R), Alpha=ScalarParameter() {parameterName=IsGlider})) {name=GreenTintMask, nodeColor=<rgba 1 0.5156 0 1>}
  n17 = PreSkinnedNormal()
  NamedRerouteDeclaration(None=ShaderStageSwitch(PixelShader=VertexInterpolator(VS=@n17).PS, VertexShader=@n17)) {name=PreskinnedNor, nodeColor=<rgba 1 0 0.6094 1>}
  n18 = PreSkinnedPosition()
  n19 = NamedRerouteDeclaration(None=ShaderStageSwitch(PixelShader=VertexInterpolator(VS=@n18).PS, VertexShader=@n18)) {name=Preskinned, nodeColor=<rgba 0.9922 0 1 1>}
  n20 = ComponentMask(None=@n19) {b=true, r=true}
  n21 = AppendVector(A=Constant(), B=ScalarParameter() {defaultValue=80, group=BadBear, parameterName=BB_PulsePreskinScale})
  n22 = SphereMask(A=@n20, B=@n21, Radius=ScalarParameter() {defaultValue=100, group=BadBear, parameterName=BB_PulsePreskinRadius}) {attenuationRadius=100, hardnessPercent=0}
  NamedRerouteDeclaration(None=@n22) {name=PreskinnedMap, nodeColor=<rgba 1 0 0.0703 1>}
  n23 = SphereMask(A=SphereMask(A=@n20, B=Constant2Vector() {g=115}) {attenuationRadius=40, hardnessPercent=0}, B=Constant() {r=0.3}, Hardness=Constant()) {attenuationRadius=0.2, hardnessPercent=1}
  n24 = ComponentMask(None=@n20) {r=true}
  n25 = Add(
   A=SphereMask(A=@n24, B=Constant() {r=65}) {attenuationRadius=15, hardnessPercent=1}
   B=SphereMask(A=@n24, B=Constant() {r=-65}) {attenuationRadius=15, hardnessPercent=10}
  )
  n26 = MaterialFunctionCall(Float3=@n19) {materialFunction=&BreakOutFloat3Components}
  n27 = MaterialFunctionCall(X=Abs(None=@n26.R), Y=@n26.G, Z=@n26.G) {materialFunction=&MakeFloat3}
  n28 = Constant3Vector() {constant=<rgba 45 -10 112 0>}
  n29 = NamedRerouteUsage()
  n30 = ComponentMask(None=@n29) {g=true}
  n31 = Constant() {r=-2}
  n32 = Abs(None=ComponentMask(None=@n29) {r=true})
  n33 = Add(
   A=Add(A=Add(A=@n23, B=@n25), B=SphereMask(A=@n27.Result, B=@n28) {attenuationRadius=8, hardnessPercent=80})
   B=Multiply(A=SphereMask(A=@n30, B=@n31) {attenuationRadius=2, hardnessPercent=0}, B=Step(Y=@n32) {constX=11})
  )
  n34 = SphereMask(A=ComponentMask(None=@n29) {b=true}, B=Constant() {r=1}) {attenuationRadius=2, hardnessPercent=20}
  NamedRerouteDeclaration(None=Saturate(None=Add(A=@n33, B=@n34))) {name=DistortionMask, nodeColor=<rgba 0.1484 0 1 1>}
  n35 = WorldPosition()
  n36 = VectorParameter() {group=BadBear, parameterName=BB_BearHeartLocation, defaultValue=<rgba 0 0 107 1>}
  n37 = ScalarParameter() {defaultValue=28, group=BadBear, parameterName=BB_Limbs_PushRadius}
  n38 = ScalarParameter() {group=BadBear, parameterName=BB_Limbs_PushHardness}
  n39 = Step(Y=ComponentMask(None=NamedRerouteUsage()) {b=true}) {constX=94}
  n40 = Saturate(None=Multiply(A=SphereMask(A=@n35.XYZ, B=@n36.RGB, Radius=@n37, Hardness=@n38) {attenuationRadius=28, hardnessPercent=0}, B=OneMinus(None=@n39)))
  n41 = WorldPosition()
  n42 = VectorParameter() {group=BadBear, parameterName=BB_BearHeartLocation_upper, defaultValue=<rgba 0 0 120 1>}
  n43 = ScalarParameter() {defaultValue=25, group=BadBear, parameterName=BB_Limbs_PushRadius_upper}
  n44 = Saturate(None=SphereMask(A=@n41.XYZ, B=@n42.RGB, Radius=@n43, Hardness=ScalarParameter() {group=BadBear, parameterName=BB_Limbs_PushHardness_upper}) {attenuationRadius=25, hardnessPercent=0})
  NamedRerouteDeclaration(None=Sign(None=Add(A=@n40, B=@n44))) {name=SocketStrength, nodeColor=<rgba 0 0.7891 1 1>}
  n45 = TextureSample() {samplerType=SAMPLERTYPE_Masks, texture=/Game/Characters/Player/Male/Large/Bodies/M_LRG_BadBear/Textures/T_Character_BadBear_WPOMask}
  NamedRerouteDeclaration(None=@n45.R) {name=CenterMask, nodeColor=<rgba 0 0.5547 1 1>}
  n46 = Multiply(A=@n37, B=Normalize(VectorInput=Subtract(A=@n35.XYZ, B=@n36.RGB)))
  n47 = Subtract(A=@n41.XYZ, B=@n42.RGB)
  n48 = Multiply(A=Multiply(A=Normalize(VectorInput=@n47), B=@n43), B=@n44)
  n49 = Constant()
  n50 = MaterialFunctionCall(
   X=ComponentMask(None=Transform(None=@n48) {transformSourceType=TRANSFORMSOURCE_World, transformType=TRANSFORM_Local}) {r=true}
   Y=@n49
   Z=@n49
  ) {materialFunction=&MakeFloat3}
  NamedRerouteDeclaration(None=Add(A=Multiply(A=@n40, B=@n46), B=Transform(None=@n50.Result) {transformSourceType=TRANSFORMSOURCE_Local})) {name=wpo, nodeColor=<rgba 0.0078 0 1 1>}
  n51 = TextureObjectParameter() {
   group=Skin FX
   parameterName=SkinFX_Mask
   samplerType=SAMPLERTYPE_Masks
   texture=/Game/Characters/Player/Male/Large/Bodies/M_LRG_BadBear/Textures/T_Character_BadBear_Body_FX
  }
  NamedRerouteDeclaration(None=@n51) {name=SkinFX_Mask, nodeColor=<rgba 0 0.6016 1 1>}
  n52 = ComponentMask(None=NamedRerouteUsage()) {g=true}
  n53 = ScalarParameter() {defaultValue=0.02, parameterName=BB_PulseNoiseAmount}
  n54 = Add(A=Time(), B=ScalarParameter() {group=BadBear, parameterName=BB_PulseDelay})
  n55 = ScalarParameter() {defaultValue=0.75, group=BadBear, parameterName=BB_PulseSpeed}
  n56 = OneMinus(None=Frac(None=Multiply(A=@n54, B=@n55)))
  n57 = SphereMask(A=@n22, B=Add(A=Multiply(A=@n52, B=@n53) {constB=0.04}, B=Power(Base=@n56) {constExponent=3})) {attenuationRadius=0.015, hardnessPercent=1}
  n58 = ScalarParameter() {defaultValue=5, parameterName=BB_PulseNoiseEms}
  n59 = ScalarParameter() {defaultValue=0.01, parameterName=BB_PulseNoiseAmount2}
  n60 = Multiply(A=@n54, B=@n55)
  n61 = OneMinus(None=Frac(None=Add(B=@n60) {constA=0.05}))
  n62 = SphereMask(A=@n22, B=Add(A=Multiply(A=@n52, B=@n59) {constB=0.04}, B=Power(Base=@n61) {constExponent=3})) {attenuationRadius=0.04, hardnessPercent=1}
  n63 = ScalarParameter() {defaultValue=10, parameterName=BB_PulseNoiseEms2}
  n64 = Multiply(
   A=Add(A=Multiply(A=@n57, B=@n58) {constB=5}, B=Multiply(A=@n62, B=@n63))
   B=SphereMask(A=@n20, B=@n21) {attenuationRadius=80, hardnessPercent=0}
  )
  n65 = Add(A=TextureCoordinate(), B=NamedRerouteUsage())
  n66 = NamedRerouteUsage()
  n67 = ScalarParameter() {defaultValue=0.25, group=BadBear, parameterName=BB_VineEms_Min}
  n68 = ScalarParameter() {defaultValue=5, group=BadBear, parameterName=BB_VineEms_Max}
  n69 = ComponentMask(None=Transform(None=Constant3Vector() {constant=<rgba 1 0 0 0>}) {transformSourceType=TRANSFORMSOURCE_Local, transformType=TRANSFORM_Camera}) {r=true}
  n70 = Multiply(
   A=LinearInterpolate(
    A=TextureSample(UVs=@n65, Tex=@n66) {samplerSource=SSM_Wrap_WorldGroupSettings}.R
    Alpha=NamedRerouteDeclaration(None=@n1.Result) {name=Fresnel, nodeColor=<rgba 0.4766 1 0 1>}
   ) {constB=0}
   B=LinearInterpolate(A=@n67, B=@n68, Alpha=Saturate(None=@n69))
  )
  n71 = Saturate(None=@n69)
  n72 = SphereMask(A=ComponentMask(None=NamedRerouteUsage()) {b=true}, B=Constant() {r=-5}) {attenuationRadius=25, hardnessPercent=10}
  n73 = Add(A=TextureCoordinate(), B=@n6)
  n74 = Multiply(
   A=LinearInterpolate(A=@n68, B=@n67, Alpha=Subtract(A=@n71, B=@n72))
   B=LinearInterpolate(A=TextureSample(UVs=@n73, Tex=@n66) {samplerSource=SSM_Wrap_WorldGroupSettings}.G, Alpha=NamedRerouteUsage()) {constB=0}
  )
  n75 = Add(
   A=Multiply(A=@n70, B=VectorParameter() {group=BadBear, parameterName=BB_VineColor1, defaultValue=<rgba 0.8912 1 0 1>}.RGB)
   B=Multiply(A=VectorParameter() {group=BadBear, parameterName=BB_VineColor2, defaultValue=<rgba 1 0.8418 0.2812 1>}.RGB, B=@n74)
  )
  n76 = Power(Base=@n75) {constExponent=0.85}
  n77 = Add(
   A=Multiply(A=@n57, B=ScalarParameter() {defaultValue=0.01, parameterName=BB_PulseNoise_UnderlyingEms}) {constB=0.01}
   B=Add(A=Multiply(A=@n64, B=@n76), B=@n75)
  )
  n78 = ScalarParameter() {defaultValue=0.4, parameterName=BB_VineHueShiftAmount}
  n79 = SphereMask(A=NamedRerouteUsage(), B=Constant()) {attenuationRadius=0.6, hardnessPercent=0}
  n80 = ScalarParameter() {defaultValue=1, parameterName=IsCharacter}
  NamedRerouteDeclaration(None=LinearInterpolate(
    A=@n77
    B=MaterialFunctionCall(Hue Shift Percentage=@n78, Texture=@n77) {materialFunction=&HueShift}.Result
    Alpha=LinearInterpolate(B=@n79, Alpha=@n80)
   )) {name=Vines, nodeColor=<rgba 1 0 0.0469 1>}
  NamedRerouteDeclaration(None=Add(A=@n70, B=@n74)) {name=VineMask, nodeColor=<rgba 0.4766 1 0 1>}
  n81 = Add(A=Divide(A=NamedRerouteUsage()) {constB=60}, B=Multiply(A=Time()) {constB=0.15})
  n82 = ComponentMask(None=@n81) {b=true, g=true}
  n83 = ComponentMask(None=@n81) {b=true, r=true}
  n84 = ComponentMask(None=NamedRerouteUsage()) {g=true}
  n85 = ComponentMask(None=MaterialFunctionCall(In=Abs(None=@n84), Contrast=ScalarParameter() {defaultValue=0.1, group=BadBear, parameterName=BB_TriPlanner_BlendStrength}) {materialFunction=&CheapContrast}.Result) {g=true}
  n86 = ConstantBiasScale(None=LinearInterpolate(
    A=TextureSample(UVs=@n82) {texture=&T_AntiMaterialCharge_Dissolve_01}.R
    B=TextureSample(UVs=@n83) {texture=&T_AntiMaterialCharge_Dissolve_01}.R
    Alpha=Saturate(None=@n85)
   )) {bias=-0.5, scale=2}
  n87 = LinearInterpolate(A=ScalarParameter() {defaultValue=0.01, parameterName=BB_VineDistortionAmount2}, Alpha=NamedRerouteUsage()) {constB=0}
  n88 = TextureSample() {
   samplerSource=SSM_Wrap_WorldGroupSettings
   samplerType=SAMPLERTYPE_Masks
   texture=&T_Character_BadBear_WPOMask
  }
  n89 = Power(Base=@n88.R) {constExponent=1.5}
  NamedRerouteDeclaration(None=Multiply(A=Multiply(A=@n86, B=@n87) {constB=0.01}, B=@n89)) {name=Distortion_1, nodeColor=<rgba 0.7109 1 0 1>}
  n90 = Add(A=Divide(A=NamedRerouteUsage()) {constB=80}, B=Multiply(A=Time()) {constB=0.2})
  n91 = ComponentMask(None=@n90) {b=true, g=true}
  n92 = ComponentMask(None=@n90) {b=true, r=true}
  n93 = ComponentMask(None=NamedRerouteUsage()) {g=true}
  n94 = ComponentMask(None=MaterialFunctionCall(In=Abs(None=@n93), Contrast=ScalarParameter() {defaultValue=0.1, group=BadBear, parameterName=BB_TriPlanner_BlendStrength}) {materialFunction=&CheapContrast}.Result) {g=true}
  n95 = ConstantBiasScale(None=LinearInterpolate(
    A=TextureSample(UVs=@n91) {texture=&T_AntiMaterialCharge_Dissolve_01}.R
    B=TextureSample(UVs=@n92) {texture=&T_AntiMaterialCharge_Dissolve_01}.R
    Alpha=Saturate(None=@n94)
   )) {bias=-0.5, scale=2}
  n96 = LinearInterpolate(A=ScalarParameter() {defaultValue=0.02, parameterName=BB_VineDistortionAmount1}, Alpha=NamedRerouteUsage()) {constB=0}
  NamedRerouteDeclaration(None=Multiply(A=@n89, B=Multiply(A=@n95, B=@n96) {constB=0.01})) {name=Distortion_2, nodeColor=<rgba 0.7109 1 0 1>}
  Reroute()
  ScalarParameter() {defaultValue=0.02, parameterName=BB_VineDistortionAmount1}
  n97 = MaterialFunctionCall() {materialFunction=&BaseFN_CharacterFunction_PreFX}
  n98 = NamedRerouteDeclaration(None=TextureSample(Tex=@n51) {samplerType=SAMPLERTYPE_Masks}.RGB) {name=FXMask, nodeColor=<rgba 1 0.7734 0 1>}
  n99 = StaticComponentMaskParameter(None=@n98) {defaultB=true, group=Skin FX, parameterName=Slurp Channel}
  n100 = StaticComponentMaskParameter(None=@n98) {defaultG=true, group=Skin FX, parameterName=Slurp Bubbles Channel}
  n101 = StaticComponentMaskParameter(None=@n98) {defaultB=true, group=Skin FX, parameterName=Slurp Normal Distortion Channel}
  n102 = GetMaterialAttributes(None=StaticSwitchParameter(
    True=MaterialFunctionCall!(Alpha=@n99, AO=@n97.AO, BubblesMask=@n100, SlurpNormalDistortionMask=@n101, Pre Skinned Position=@n97.AO, Material=@n97.AO) {materialFunction=/Game/Weapons/WeaponSkins/MaterialFunction/MF_Slurp}.Result
    False=@n97.Result
   ) {group=Skin FX, parameterName=UseSlurp}) {attributeGetTypes [#1, #2, #3]}
  n103 = NamedRerouteUsage()
  n104 = Power(Base=@n45.R) {constExponent=1}
  n105 = OneMinus(None=ScalarParameter() {defaultValue=1, parameterName=BB_CenterDarkenAmount})
  n106 = LinearInterpolate(B=Saturate(None=Add(A=@n104, B=@n105)), Alpha=ScalarParameter() {defaultValue=1, parameterName=IsCharacter}) {constA=1}
  n107 = LinearInterpolate(
   A=LinearInterpolate(
    A=Multiply(A=@n102.Emissive Color, B=@n106)
    B=VectorParameter() {group=BadBear, parameterName=HandFootColors, defaultValue=<rgba 0 0.8021 0.2724 1>}.RGB
    Alpha=NamedRerouteUsage()
   )
   Alpha=NamedRerouteUsage()
  ) {constB=0}
  n108 = NamedRerouteUsage()
  n109 = MaterialFunctionCall(
   AO=@n97.AO
   Distance Blend=@n97.AO
   Fresnel=@n97.AO
   Gmap=@n97.AO
   Pre Skinned Local Position=@n97.AO
   Material=SetMaterialAttributes!(
    MaterialAttributes=@n102.MaterialAttributes
    World Position Offset=Add(A=@n102.World Position Offset, B=@n103)
    Emissive Color=Add(A=@n107, B=@n108)
   ) {attributeSetTypes [#3, #2]}
  ) {materialFunction=&BaseFN_CharacterFunction_PostFX}
  n110 = MaterialFunctionCall(In=@n109.Result) {materialFunction=&CINE_MF_CharPostFX}
  n111 = Constant()
  n112 = SetMaterialAttributes!(MaterialAttributes=@n110.Result, Base Color=@n111, Metallic=@n111, Specular=@n111, Roughness=@n111, Emissive Color=NamedRerouteUsage()) {attributeSetTypes [
    #1
    #4
    #5
    #6
    #2
   ]}
  StaticSwitchParameter(True=@n112, False=@n110.Result) {parameterName=DEBUG_Socket}
  TextureSample() {samplerType=SAMPLERTYPE_Grayscale, texture=/FortWater/Textures/LowResBlurredNoise}
 ]
}
