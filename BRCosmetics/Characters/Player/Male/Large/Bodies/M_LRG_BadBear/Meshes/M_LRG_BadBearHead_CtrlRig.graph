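# Control Rig for the large male Bad Bear head cosmetic that adjusts the MaskCheat bone, a child of head, according to head pose.
# A Spherical Pose Reader tracks the head bone's Y axis within an asymmetric active region with falloff.
# Its output weights a Modify Transforms step that applies a fixed Euler Z rotation to MaskCheat, so this corrective blends in only at certain head orientations.
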ControlRigBlueprint /BRCosmetics/Characters/Player/Male/Large/Bodies/M_LRG_BadBear/Meshes/M_LRG_BadBearHead_CtrlRig
graphs [
 RigVMFunctionLibrary {links [], nodes []}
 RigVMModel {nodes [
   RigUnit_BeginExecution = &RigUnit_BeginExecution()
   n1 = &RigUnit_SphericalPoseReader(
    ActiveRegionScaleFactors=(PositiveWidth=0.05,NegativeWidth=0.1,PositiveHeight=0.5,NegativeHeight=0.5)
    ActiveRegionSize=1
    Debug=(bDrawDebug=false,DebugScale=25,DebugSegments=20,DebugThickness=0.25)
    DriverAxis=(X=0,Y=1,Z=0)
    DriverItem=(Type=Bone,Name="head")
    FalloffRegionScaleFactors=(PositiveWidth=0.1,NegativeWidth=0.1,PositiveHeight=0.5,NegativeHeight=0.5)
    FalloffSize=0.2
    OptionalParentItem=(Type=Bone)
    RotationOffset=(X=90,Y=0,Z=30)
    ExecutePin=@RigUnit_BeginExecution.ExecutePin
   )
   > &RigUnit_ModifyTransforms(
    ItemToModify=((Item=(Type=Bone,Name="MaskCheat"),Transform=(Rotation=(X=0,Y=0,Z=0,W=1),Translation=(X=0,Y=0,Z=0),Scale3D=(X=1,Y=1,Z=1))))
    ItemToModify.0.Transform.Rotation=&RigVMFunction_MathQuaternionFromEuler(Euler=(X=0,Y=0,Z=-25)).Result
    Weight=@n1.OutputParam
   )
  ]}
]
hierarchy [
 root {loc=<0 0 0>, rot=<0 0 0 1>}
 pelvis {parent=root, loc=<0 -0.7321 92.4218>, rot=<0.0127 0.707 -0.0127 -0.707>}
 spine_01 {parent=pelvis, loc=<5.6002 -0.5128 0>, rot=<0 0 0.0532 -0.9986>}
 spine_02 {parent=spine_01, loc=<8.0718 -0.0114 0>, rot=<0 0 0.0014 -1>}
 spine_03 {parent=spine_02, loc=<7.0362 -0.0083 0>, rot=<0 0 0.0012 -1>}
 spine_04 {parent=spine_03, loc=<8.0992 -0.0104 0>, rot=<0 0 0.0013 -1>}
 spine_05 {parent=spine_04, loc=<10.0112 -0.015 0>, rot=<0 0 0.001 -1>}
 clavicle_l {parent=spine_05, loc=<2.5856 -0.2866 -4.3777>, rot=<0.0449 -0.7291 -0.0127 -0.6828>}
 upperarm_l {parent=clavicle_l, loc=<18.2269 0 0>, rot=<0.0231 -0.2805 0.0014 -0.9596>}
 lowerarm_l {parent=upperarm_l, loc=<29.0284 0 0>, rot=<0 0 -0.2475 -0.9689>}
 hand_l {parent=lowerarm_l, loc=<25.8757 0 0>, rot=<-0.6913 -0.0958 0.1331 -0.7037>}
 weapon_l {parent=hand_l, loc=<-0.1974 -3.4111 0.3811>, rot=<0 0 0 1>}
 thumb_01_l {parent=hand_l, loc=<1.3118 -0.3191 -2.459>, rot=<0.6546 -0.4042 0.0664 -0.6354>}
 thumb_02_l {parent=thumb_01_l, loc=<5.9336 -0.0448 0.0401>, rot=<-0.0088 0.005 0.2979 -0.9545>}
 thumb_03_l {parent=thumb_02_l, loc=<2.2223 0.3682 0>, rot=<0 0 0.0067 -1>}
 index_metacarpal_l {parent=hand_l, loc=<1.6535 0.0183 -2.96>, rot=<0.0158 -0.0124 -0.0173 -0.9996>}
 index_01_l {parent=index_metacarpal_l, loc=<8.3008 0.0357 -0.0524>, rot=<-0.0044 0.0485 0.0924 -0.9945>}
 index_02_l {parent=index_01_l, loc=<3.3587 -0.3004 0>, rot=<0 0 0.3225 -0.9466>}
 index_03_l {parent=index_02_l, loc=<2.5578 -0.123 0>, rot=<0 0 0.2559 -0.9667>}
 middle_metacarpal_l {parent=hand_l, loc=<2.1856 -0.1725 -0.5984>, rot=<0.006 0.0168 -0.0329 -0.9993>}
 middle_01_l {parent=middle_metacarpal_l, loc=<7.3275 -0.2447 0.0129>, rot=<-0.0208 0.0042 0.0858 -0.9961>}
 middle_02_l {parent=middle_01_l, loc=<4.016 0.1415 0>, rot=<0 0 0.3519 -0.936>}
 middle_03_l {parent=middle_02_l, loc=<2.6103 0.0517 0>, rot=<0 0 0.1707 -0.9853>}
 ring_metacarpal_l {parent=hand_l, loc=<2.2074 -0.3625 1.3638>, rot=<-0.0284 0.0722 -0.0458 -0.9959>}
 ring_01_l {parent=ring_metacarpal_l, loc=<7.0276 -0.6134 -0.2041>, rot=<-0.0484 -0.0225 0.1894 -0.9805>}
 ring_02_l {parent=ring_01_l, loc=<3.6174 0.4647 0>, rot=<0 0 0.2806 -0.9598>}
 ring_03_l {parent=ring_02_l, loc=<2.4152 0.114 0>, rot=<0 0 0.1189 -0.9929>}
 pinky_metacarpal_l {parent=hand_l, loc=<1.7616 -0.9753 3.3814>, rot=<-0.0397 0.0814 -0.069 -0.9935>}
 pinky_01_l {parent=pinky_metacarpal_l, loc=<7.1709 -0.4831 -0.1094>, rot=<-0.0676 -0.0132 0.2236 -0.9723>}
 pinky_02_l {parent=pinky_01_l, loc=<2.8274 0.0893 0>, rot=<0 0 0.2854 -0.9584>}
 pinky_03_l {parent=pinky_02_l, loc=<2.2046 0.0307 0>, rot=<0 0 0.1445 -0.9895>}
 hand_attach_l {parent=hand_l, loc=<8.5 -3 0>, rot=<0.7071 0.7071 0 0>}
 deform_wrist_dn_base_l {parent=hand_l, loc=<0 -3.4173 0>, rot=<0 0 0 -1>}
 deform_wrist_dn_l {parent=deform_wrist_dn_base_l, loc=<0 -0.1177 0>, rot=<0 0 0 1>, scale=<0.9941 1 1>}
 deform_wrist_up_base_l {parent=hand_l, loc=<0 2.7805 0>, rot=<0 0 0 -1>}
 deform_wrist_up_l {parent=deform_wrist_up_base_l, loc=<0 0.0004 0>, rot=<0 0 0 1>}
 lowerarm_twist_02_l {parent=lowerarm_l, loc=<9.6199 0 0>, rot=<0 0 0 1>}
 lowerarm_twist_01_l {parent=lowerarm_l, loc=<17.8549 0 0>, rot=<0 0 0 1>}
 deform_elbow_in_base_l {parent=lowerarm_l, loc=<0.2841 5.6899 -0.1986>, rot=<0.0396 -0.0086 0.0055 -0.9992>}
 deform_elbow_in_l {parent=deform_elbow_in_base_l, loc=<0 0 0>, rot=<0 0 0 1>}
 deform_elbow_out_base_l {parent=lowerarm_l, loc=<-0.0776 -7.4888 0.0235>, rot=<0.0396 -0.0086 0.0055 -0.9992>}
 deform_elbow_out_l {parent=deform_elbow_out_base_l, loc=<0 0 0>, rot=<0 0 0 1>}
 lowerarm_l_single {parent=lowerarm_l, loc=<0 0 0>, rot=<0 0 0 1>}
 upperarm_twist_01_l {parent=upperarm_l, loc=<0.0041 0 0>, rot=<0 0 0 1>}
 upperarm_twist_01_l_single {parent=upperarm_twist_01_l, loc=<0 0 0>, rot=<0 0 0 1>}
 deform_upperarm_bk_base_l {parent=upperarm_twist_01_l, loc=<1.0436 -13.2993 -1.957>, rot=<-0.0014 0.0115 0.0029 -0.9999>}
 deform_upperarm_bk_l {parent=deform_upperarm_bk_base_l, loc=<0 0 0>, rot=<0 0 0 1>}
 deform_upperarm_fr_base_l {parent=upperarm_twist_01_l, loc=<-0.2462 10.8052 -2.7427>, rot=<-0.0014 0.0115 0.0029 -0.9999>}
 deform_upperarm_fr_l {parent=deform_upperarm_fr_base_l, loc=<0 0 0>, rot=<0 0 0 1>}
 deform_upperarm_up_base_l {parent=upperarm_twist_01_l, loc=<-3.419 -1.1416 8.0796>, rot=<-0.0014 0.0115 0.0029 -0.9999>}
 deform_upperarm_up_l {parent=deform_upperarm_up_base_l, loc=<0 0 0>, rot=<0 0 0 1>}
 upperarm_twist_02_l {parent=upperarm_l, loc=<13.7941 0 0>, rot=<0 0 0 1>}
 deform_bicep_base_l {parent=upperarm_twist_02_l, loc=<5.7682 8.0196 -0.1488>, rot=<-0.0014 0.0115 0.0029 -0.9999>}
 deform_bicep_l {parent=deform_bicep_base_l, loc=<0 0 0>, rot=<0 0 0 1>}
 upperarm_l_single {parent=upperarm_l, loc=<0 0 0>, rot=<0 0 0 1>}
 neck_01 {parent=spine_05, loc=<10.815 1.9196 0>, rot=<0 0 -0.1864 -0.9825>}
 neck_02 {parent=neck_01, loc=<4.7347 0.3045 0>, rot=<0 0 0.0363 -0.9993>}
 head {parent=neck_02, loc=<5.7632 0.5552 0>, rot=<0 0 0.115 -0.9934>}
 MaskCheat {parent=head, loc=<0 0 0>, rot=<0 0 0 1>}
 clavicle_r {parent=spine_05, loc=<2.5855 -0.2866 4.3777>, rot=<0.7291 0.0449 -0.6828 0.0127>}
 upperarm_r {parent=clavicle_r, loc=<-18.2269 0 0>, rot=<0.0231 -0.2805 0.0014 -0.9596>}
 lowerarm_r {parent=upperarm_r, loc=<-29.0284 0 0>, rot=<0 0 -0.2475 -0.9689>}
 hand_r {parent=lowerarm_r, loc=<-25.8757 0 0>, rot=<-0.6913 -0.0958 0.1331 -0.7037>}
 weapon_r {parent=hand_r, loc=<0.1974 3.4111 -0.3811>, rot=<0 0 0 1>}
 thumb_01_r {parent=hand_r, loc=<-1.3118 0.3191 2.459>, rot=<0.6546 -0.4042 0.0664 -0.6354>}
 thumb_02_r {parent=thumb_01_r, loc=<-5.9336 0.0448 -0.0401>, rot=<-0.0088 0.005 0.2979 -0.9545>}
 thumb_03_r {parent=thumb_02_r, loc=<-2.2223 -0.3682 0>, rot=<0 0 0.0067 -1>}
 index_metacarpal_r {parent=hand_r, loc=<-1.6535 -0.0183 2.96>, rot=<0.0158 -0.0124 -0.0173 -0.9996>}
 index_01_r {parent=index_metacarpal_r, loc=<-8.3008 -0.0357 0.0524>, rot=<-0.0044 0.0485 0.0924 -0.9945>}
 index_02_r {parent=index_01_r, loc=<-3.3587 0.3004 0>, rot=<0 0 0.3225 -0.9466>}
 index_03_r {parent=index_02_r, loc=<-2.5578 0.123 0>, rot=<0 0 0.2559 -0.9667>}
 middle_metacarpal_r {parent=hand_r, loc=<-2.1856 0.1725 0.5984>, rot=<0.006 0.0168 -0.0329 -0.9993>}
 middle_01_r {parent=middle_metacarpal_r, loc=<-7.3275 0.2447 -0.0129>, rot=<-0.0208 0.0042 0.0858 -0.9961>}
 middle_02_r {parent=middle_01_r, loc=<-4.016 -0.1415 0>, rot=<0 0 0.3519 -0.936>}
 middle_03_r {parent=middle_02_r, loc=<-2.6103 -0.0517 0>, rot=<0 0 0.1707 -0.9853>}
 ring_metacarpal_r {parent=hand_r, loc=<-2.2074 0.3625 -1.3638>, rot=<-0.0284 0.0722 -0.0458 -0.9959>}
 ring_01_r {parent=ring_metacarpal_r, loc=<-7.0276 0.6134 0.2041>, rot=<-0.0484 -0.0225 0.1894 -0.9805>}
 ring_02_r {parent=ring_01_r, loc=<-3.6174 -0.4647 0>, rot=<0 0 0.2806 -0.9598>}
 ring_03_r {parent=ring_02_r, loc=<-2.4152 -0.114 0>, rot=<0 0 0.1189 -0.9929>}
 pinky_metacarpal_r {parent=hand_r, loc=<-1.7616 0.9753 -3.3814>, rot=<-0.0397 0.0814 -0.069 -0.9935>}
 pinky_01_r {parent=pinky_metacarpal_r, loc=<-7.1709 0.4831 0.1094>, rot=<-0.0676 -0.0132 0.2236 -0.9723>}
 pinky_02_r {parent=pinky_01_r, loc=<-2.8274 -0.0893 0>, rot=<0 0 0.2854 -0.9584>}
 pinky_03_r {parent=pinky_02_r, loc=<-2.2046 -0.0307 0>, rot=<0 0 0.1445 -0.9895>}
 hand_attach_r {parent=hand_r, loc=<-8.5 3 0>, rot=<0 0 -0.7071 -0.7071>}
 deform_wrist_dn_base_r {parent=hand_r, loc=<0 3.4172 0>, rot=<0 0 0 -1>}
 deform_wrist_dn_r {parent=deform_wrist_dn_base_r, loc=<0 0.1177 0>, rot=<0 0 0 -1>, scale=<0.9941 1 1>}
 deform_wrist_up_base_r {parent=hand_r, loc=<-0.0001 -2.7804 0>, rot=<0 0 0 -1>}
 deform_wrist_up_r {parent=deform_wrist_up_base_r, loc=<0 -0.0004 0>, rot=<0 0 0 -1>}
 lowerarm_twist_02_r {parent=lowerarm_r, loc=<-9.6199 0 0>, rot=<0 0 0 -1>}
 lowerarm_twist_01_r {parent=lowerarm_r, loc=<-17.8549 0 0>, rot=<0 0 0 -1>}
 deform_elbow_in_base_r {parent=lowerarm_r, loc=<-0.2843 -5.6898 0.1989>, rot=<0.0396 -0.0086 0.0055 -0.9992>}
 deform_elbow_in_r {parent=deform_elbow_in_base_r, loc=<0 0 0>, rot=<0 0 0 1>}
 deform_elbow_out_base_r {parent=lowerarm_r, loc=<0.0775 7.4889 -0.0233>, rot=<0.0396 -0.0086 0.0055 -0.9992>}
 deform_elbow_out_r {parent=deform_elbow_out_base_r, loc=<0 0 0>, rot=<0 0 0 1>}
 lowerarm_r_single {parent=lowerarm_r, loc=<0 0 0>, rot=<0 0 0 1>}
 upperarm_twist_01_r {parent=upperarm_r, loc=<-0.0041 0 0>, rot=<0 0 0 -1>}
 deform_upperarm_bk_base_r {parent=upperarm_twist_01_r, loc=<-1.0435 13.2993 1.9568>, rot=<-0.0014 0.0115 0.0029 -0.9999>}
 deform_upperarm_bk_r {parent=deform_upperarm_bk_base_r, loc=<0 0 0>, rot=<0 0 0 1>}
 deform_upperarm_fr_base_r {parent=upperarm_twist_01_r, loc=<0.2461 -10.8053 2.7428>, rot=<-0.0014 0.0115 0.0029 -0.9999>}
 deform_upperarm_fr_r {parent=deform_upperarm_fr_base_r, loc=<0 0 0>, rot=<0 0 0 1>}
 deform_upperarm_up_base_r {parent=upperarm_twist_01_r, loc=<3.4192 1.1417 -8.08>, rot=<-0.0014 0.0115 0.0029 -0.9999>}
 deform_upperarm_up_r {parent=deform_upperarm_up_base_r, loc=<0 0 0>, rot=<0 0 0 1>}
 upperarm_twist_01_r_single {parent=upperarm_twist_01_r, loc=<0 0 0>, rot=<0 0 0 1>}
 upperarm_twist_02_r {parent=upperarm_r, loc=<-13.7941 0 0>, rot=<0 0 0 -1>}
 deform_bicep_base_r {parent=upperarm_twist_02_r, loc=<-5.768 -8.0196 0.1485>, rot=<-0.0014 0.0115 0.0029 -0.9999>}
 deform_bicep_r {parent=deform_bicep_base_r, loc=<0 0 0>, rot=<0 0 0 1>}
 upperarm_r_single {parent=upperarm_r, loc=<0 0 0>, rot=<0 0 0 1>}
 deform_pec_aim_base_l {parent=spine_04, loc=<5.0875 13.6814 -2.6508>, rot=<0.066 -0.5203 0.3247 -0.7871>}
 deform_pec_aim_l {parent=deform_pec_aim_base_l, loc=<0 0 0>, rot=<0 0 0 1>}
 deform_pec_base_l {parent=deform_pec_aim_l, loc=<16.3279 0 0>, rot=<0 0 0 1>}
 deform_pec_l {parent=deform_pec_base_l, loc=<0 0 0>, rot=<0 0 0 1>}
 deform_pec_aim_base_r {parent=spine_04, loc=<5.0873 13.6814 2.6508>, rot=<-0.5203 -0.066 0.7871 0.3247>}
 deform_pec_aim_r {parent=deform_pec_aim_base_r, loc=<0 0 0>, rot=<0 0 0 1>}
 deform_pec_base_r {parent=deform_pec_aim_r, loc=<-16.3279 0 0>, rot=<0 0 0 -1>}
 deform_pec_r {parent=deform_pec_base_r, loc=<0 0 0.0002>, rot=<0 0 0 -1>}
 deform_lat_l_base {parent=spine_03, loc=<2.2029 -10.154 -17.6918>, rot=<0 0 0.0169 -0.9999>}
 deform_lat_l {parent=deform_lat_l_base, loc=<0 0 0>, rot=<0 0 0 1>}
 deform_lat_r_base {parent=spine_03, loc=<2.203 -10.1541 17.6918>, rot=<0 0 0.9999 0.0169>}
 deform_lat_r {parent=deform_lat_r_base, loc=<0 0 0>, rot=<0 0 0 -1>}
 BearBase {parent=spine_01, loc=<0 0 0>, rot=<0 0 0 1>}
 thigh_l {parent=pelvis, loc=<-1.8716 -0.0109 -11.9936>, rot=<0.0013 0.0474 0.0128 -0.9988>}
 calf_l {parent=thigh_l, loc=<-41.9121 0 0>, rot=<0 0 -0.0203 -0.9998>}
 foot_l {parent=calf_l, loc=<-38.7196 0 0>, rot=<0.0992 -0.052 0.0151 -0.9936>}
 ball_l {parent=foot_l, loc=<-6.1614 13.1469 0.6807>, rot=<0.0001 -0.0001 -0.7063 -0.7079>}
 calf_twist_02_l {parent=calf_l, loc=<-18.8064 -0.0967 0.0397>, rot=<0.0237 0.0095 -0.0035 -0.9997>}
 calf_twist_01_l {parent=calf_l, loc=<-35.813 -0.0967 0.0397>, rot=<0.0237 0.0095 -0.0035 -0.9997>}
 deform_knee_in_base_l {parent=calf_l, loc=<1.555 -7.9899 -0.7855>, rot=<0.0008 -0.0063 -0.0161 -0.9999>}
 deform_knee_in_l {parent=deform_knee_in_base_l, loc=<0 0 0>, rot=<0 0 0 1>}
 deform_knee_out_base_l {parent=calf_l, loc=<-0.221 6.8704 -0.7392>, rot=<0.0008 -0.0063 -0.0161 -0.9999>}
 deform_knee_out_l {parent=deform_knee_out_base_l, loc=<0 0 0>, rot=<0 0 0 1>}
 thigh_twist_01_l {parent=thigh_l, loc=<-21.629 -0.0919 -0.0002>, rot=<0.0013 0 0 -1>}
 deform_groin_base_l {parent=thigh_l, loc=<-2.1102 8.578 -1.7847>, rot=<-0.0001 0.0098 0.0027 -0.9999>}
 deform_groin_l {parent=deform_groin_base_l, loc=<0 0 0>, rot=<0 0 0 1>}
 thigh_r {parent=pelvis, loc=<-1.8716 -0.0109 11.9936>, rot=<0.0474 -0.0013 0.9988 0.0128>}
 calf_r {parent=thigh_r, loc=<41.9121 0 0>, rot=<0 0 -0.0203 -0.9998>}
 foot_r {parent=calf_r, loc=<38.7196 0 0>, rot=<0.0992 -0.052 0.0151 -0.9936>}
 ball_r {parent=foot_r, loc=<6.1614 -13.1469 -0.6807>, rot=<0.0001 -0.0001 -0.7063 -0.7079>}
 calf_twist_02_r {parent=calf_r, loc=<18.8064 0.0967 -0.0397>, rot=<0.0237 0.0095 -0.0035 -0.9997>}
 calf_twist_01_r {parent=calf_r, loc=<35.813 0.0967 -0.0397>, rot=<0.0237 0.0095 -0.0035 -0.9997>}
 deform_knee_in_base_r {parent=calf_r, loc=<-1.555 7.9899 0.7855>, rot=<0.0008 -0.0063 -0.0161 -0.9999>}
 deform_knee_in_r {parent=deform_knee_in_base_r, loc=<0 0 0>, rot=<0 0 0 1>}
 deform_knee_out_base_r {parent=calf_r, loc=<0.221 -6.8704 0.7391>, rot=<0.0008 -0.0063 -0.0161 -0.9999>}
 deform_knee_out_r {parent=deform_knee_out_base_r, loc=<0 0 0>, rot=<0 0 0 1>}
 thigh_twist_01_r {parent=thigh_r, loc=<21.629 0.0919 0.0002>, rot=<0.0013 0 0 -1>}
 deform_groin_base_r {parent=thigh_r, loc=<2.1102 -8.578 1.7848>, rot=<-0.0001 0.0098 0.0027 -0.9999>}
 deform_groin_r {parent=deform_groin_base_r, loc=<0 0 0>, rot=<0 0 0 -1>}
 deform_glute_base_l {parent=pelvis, loc=<-10.1877 -11.6217 -12.899>, rot=<0.0275 0.0501 -0.4472 -0.8926>}
 deform_glute_l {parent=deform_glute_base_l, loc=<0 -0.0671 0>, rot=<0 0 0 1>}
 deform_glute_base_r {parent=pelvis, loc=<-10.1877 -11.6218 12.899>, rot=<0.0501 -0.0275 0.8926 -0.4472>}
 deform_glute_r {parent=deform_glute_base_r, loc=<0 0.0671 0>, rot=<0 0 0 1>}
 attach {parent=root, loc=<0 0 0>, rot=<0 0 0 1>}
 ArmUpPistol {type=Curve, loc=<0 0 0>, rot=<0 0 0 1>}
]
