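# Animation Blueprint for the BadBear cosmetic body, driving its deformation on the player character.
# It copies the parent mesh pose through deform and head Control Rigs, then adds pinky rotation and a curve-driven upper-arm pistol fix outside skydiving.
# After evaluation, it feeds two WPO socket locations into material vector parameters.
# Spring-bone and clavicle nodes sit unconnected to the output pose, leaving the computed skydiving alpha without visible effect.
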AnimBlueprint /BRCosmetics/Characters/Player/Male/Large/Bodies/M_LRG_BadBear/Meshes/M_LRG_BadBear_AnimBP
target_skeleton=&/M_LRG_BadBear_Skeleton
graphs [
 AnimGraph {nodes [
   Component To Local(ComponentPose=Spring controller - Bone: bear_spine_02!(ComponentPose:Component Space Pose Link Structure, Alpha:float=Get skydivingFloat() -> (skydivingFloat:double)) -> (Pose:Component Space Pose Link Structure) {
     is=title AnimGraphNode_SpringBone
     maxDisplacement=5
     springStiffness=85
     alphaBoolBlend {blendOption=Linear}
     springBone {boneName=bear_spine_02}
    })
   "Transform (Modify) Bone - Bone: bear_clavicle_r"(Alpha=1, Translation="0,-10,0", Rotation="0,-20,0", Scale="1,1,1") {
    rotationMode=BMM_Additive
    rotationSpace=BCS_BoneSpace
    scaleMode=BMM_Additive
    scaleSpace=BCS_BoneSpace
    translationMode=BMM_Additive
    alphaBoolBlend {blendOption=Linear}
    alphaScaleBiasClamp {bInterpResult=true, interpSpeedIncreasing=500}
    boneToModify {boneName=bear_clavicle_r}
    rotation {yaw=-20}
    translation {y=-10}
   }
   "Transform (Modify) Bone - Bone: pinky_01_r"!(ComponentPose, bAlphaBoolEnabled:bool=True, Translation="0,0,0", Rotation="0,-20,0", Scale="1,1,1") -> (Pose) {
    alphaInputType=Bool
    rotationMode=BMM_Additive
    rotationSpace=BCS_BoneSpace
    scaleMode=BMM_Additive
    scaleSpace=BCS_BoneSpace
    translationMode=BMM_Additive
    alphaBoolBlend {blendOption=Linear}
    alphaScaleBiasClamp {bInterpResult=true, interpSpeedIncreasing=500}
    boneToModify {boneName=pinky_01_r}
    rotation {yaw=-20}
   }
   n1 = Control Rig(Source=Copy Pose From Mesh() {bUseAttachedParent=true}) {
    alphaBoolBlend {blendOption=Linear}
    controlRigAssetReference {blueprintRigClass=&/M_LRG_BadBear_DeformRig_ControlRig.M_LRG_BadBear_DeformRig_ControlRig_C}
   }
   n2 = Control Rig(Source=@n1) {
    alphaBoolBlend {blendOption=Linear}
    controlRigAssetReference {blueprintRigClass=&/M_LRG_BadBearHead_CtrlRig.M_LRG_BadBearHead_CtrlRig_C}
   }
   n3 = "Transform (Modify) Bone - Bone: deform_upperarm_up_base_r"(
    ComponentPose=Local To Component(LocalPose=@n2)
    Alpha=Get pistolPoseFix() -> (pistolPoseFix:double)
    Translation="20,0,0"
    Rotation="0,0,0"
    Scale="1,1,1"
   ) {
    rotationMode=BMM_Additive
    rotationSpace=BCS_BoneSpace
    scaleMode=BMM_Additive
    scaleSpace=BCS_BoneSpace
    translationMode=BMM_Additive
    alphaBoolBlend {blendOption=Linear}
    alphaScaleBiasClamp {bInterpResult=true, interpSpeedIncreasing=500}
    boneToModify {boneName=deform_upperarm_up_base_r}
    translation {x=20}
   }
   n4 = "Transform (Modify) Bone - Bone: pinky_01_l"!(ComponentPose=@n3, bAlphaBoolEnabled:bool=True, Translation="0,0,0", Rotation="0,-20,0", Scale="1,1,1") -> (Pose) {
    alphaInputType=Bool
    rotationMode=BMM_Additive
    rotationSpace=BCS_BoneSpace
    scaleMode=BMM_Additive
    scaleSpace=BCS_BoneSpace
    translationMode=BMM_Additive
    alphaBoolBlend {blendOption=Linear}
    alphaScaleBiasClamp {bInterpResult=true, interpSpeedIncreasing=500}
    boneToModify {boneName=pinky_01_l}
    rotation {yaw=-20}
   }
   n5 = "Transform (Modify) Bone - Bone: pinky_01_r"(ComponentPose=@n4, bAlphaBoolEnabled=True, Translation="0,0,0", Rotation="0,-20,0", Scale="1,1,1") {
    alphaInputType=Bool
    rotationMode=BMM_Additive
    rotationSpace=BCS_BoneSpace
    scaleMode=BMM_Additive
    scaleSpace=BCS_BoneSpace
    translationMode=BMM_Additive
    alphaBoolBlend {blendOption=Linear}
    alphaScaleBiasClamp {bInterpResult=true, interpSpeedIncreasing=500}
    boneToModify {boneName=pinky_01_r}
    rotation {yaw=-20}
   }
   Output Pose(Result=Component To Local(ComponentPose=@n5)) {layerGroup=None}
  ]}
 EventGraph {nodes [
   n1 = Gravity Override State Function DEPRECATED(CharacterPartAnimInstance=Self-Reference(), GameTimeInSeconds=GetGameTimeInSeconds(), GravityOverrideParams=Get Gravity Override Params())
   > Set GravityOverride(GravityOverride=@n1)
   Event BlueprintUpdateAnimation()
   > n2 = Sequence!(execute) -> (then_0, then_1, then_2:Exec)
   > n3 = Cast To FortPlayerAnimInstance(Object=GetAnimInstance(self=GetOwnerSkeletalMeshComponent()))
   > n4 = Set isPistol(isPistol="Equal (Enum)"(A:EFortWeaponCoreAnimation Enum=Get WeaponCoreAnim(self=@n3), B:EFortWeaponCoreAnimation Enum=Pistol))
   n5 = Get bIsSkyDiving()
   > n6 = Branch(Condition=AND Boolean(A=@n4, B=NOT Boolean(A=@n5)))
   > Set pistolPoseFix(pistolPoseFix:double=GetCurveValue(CurveName=ArmUpPistol)) -> (Output_Get:double)
   Set pistolPoseFix(execute=@n6.else, pistolPoseFix=0)
   n7 = Branch(execute=@n2.then_1, Condition=Get bIsSkyDiving())
   > Set skydivingFloat(skydivingFloat:double=0) -> (Output_Get:double)
   Set skydivingFloat(execute=@n7.else, skydivingFloat=0.03)
   Event BlueprintPostEvaluateAnimation() {is=title K2Node_Event}
   n8 = GetPartSkeletalMeshComponent()
   > Is Valid(InputObject=@n8)
   > Delay(Duration=0)
   > SetVectorParameterValueOnMaterials!(
    execute
    self:Mesh Component Object Reference=@n8
    ParameterName:Name=BB_BearHeartLocation
    ParameterValue:Vector=GetSocketLocation!(self:Scene Component Object Reference=@n8, InSocketName:Name=WPOSocket) -> (ReturnValue:Vector)
   ) -> (then)
   > SetVectorParameterValueOnMaterials(self=@n8, ParameterName=BB_BearHeartLocation_upper, ParameterValue=GetSocketLocation(self=@n8, InSocketName=WPOSocket_upper))
   K2Node_Event_2 = CustomEvent() {is=title K2Node_CustomEvent}
   > n9 = Set SkelMesh(SkelMesh=GetPartSkeletalMeshComponent())
   > n10 = Get Dynamic Material Instance(SkeletalMeshComponent=@n9)
   > Set MID(MID:Material Instance Dynamic Object Reference=@n10) -> (Output_Get:Material Instance Dynamic Object Reference)
   Event OnCharacterCustomizationLoaded()
   > Do Once()
   n11 = CheckForBackbling()
   > Set isWearingBackbling(isWearingBackbling=OR Boolean!(A=@n11, B=@n11.bIsBackblingACape, C:bool=@n11.bIsBackblingLong) -> (ReturnValue))
   FloatSpringInterp!(
    execute
    Current:float=0
    Target:float=0
    SpringState:"Float Spring State Structure (by ref)"
    Stiffness:float=0
    CriticalDampingFactor:float=0
    DeltaTime:float=0
    Mass:float=1
    TargetVelocityAmount:float=1
    bClamp:bool=false
    MinValue:float=-1
    MaxValue:float=1
    bInitializeFromTarget:bool=false
   ) -> (then, ReturnValue:float)
   SetScalarParameterValue(self=Get MID() -> (MID:Material Instance Dynamic Object Reference), ParameterName=BB_Limbs_PushRadius, Value=0)
  ]}
]
variables [
 Gravity Override Params {type=GravityOverrideParamsStruct}
 GravityOverride {type=Vector}
 isPistol {type=boolean}
 isWearingBackbling {type=boolean}
 MID {type=MaterialInstanceDynamic}
 pistolPoseFix {type=number}
 SkelMesh {type=SkeletalMeshComponent}
 skydivingFloat {type=number}
]
