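# Control Rig for the BadBear FaceAcc part, posing a small secondary bear skeleton (bb_ bones) parented to the player's pelvis.
# It position-constrains bb_mainAim to the pelvis, partially aims bb_main at the head using that bone as secondary target, blends in spine_02 rotation, then applies a fixed offset.
# Stretch-enabled two-bone IK then solves both bear legs toward pelvis-parented foot and pole-vector controls, keeping them attached to the wearer's body.
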ControlRigBlueprint /BRCosmetics/Characters/Player/Male/Large/Bodies/M_LRG_BadBear/Meshes/Parts/M_LRG_BadBear_FaceAcc_CtrlRig
graphs [
 RigVMFunctionLibrary {links [], nodes []}
 RigVMModel {nodes [
   RigUnit_BeginExecution = &RigUnit_BeginExecution()
   /Script/ControlRig.RigUnit_PositionConstraintLocalSpaceOffset(
    Child=(Type=Bone,Name="bb_mainAim")
    Filter=(bX=True,bY=True,bZ=True)
    Parents=((Item=(Type=Bone,Name="pelvis"),Weight=1),(Item=(Type=Bone,Name="root"),Weight=0))
    Weight=1
    bMaintainOffset=True
    ExecutePin=@RigUnit_BeginExecution.ExecutePin
   ) {is=struct RigVMUnitNode}
   > /Script/ControlRig.RigUnit_AimItem(
    DebugSettings=(Scale=10,WorldOffset=(Rotation=(X=0,Y=0,Z=0,W=1),Translation=(X=0,Y=0,Z=0),Scale3D=(X=1,Y=1,Z=1)))
    Item=(Type=Bone,Name="bb_main")
    Primary=(Weight=0.6,Axis=(X=1,Y=0,Z=0),Target=(X=0,Y=0,Z=0),Kind=Location,Space=(Type=Bone,Name="head"))
    Secondary=(Weight=1,Axis=(X=0,Y=1,Z=0),Target=(X=1,Y=0,Z=0),Kind=Location,Space=(Type=Bone,Name="bb_mainAim"))
    Weight=1
   ) {is=struct RigVMUnitNode}
   > &RigUnit_RotationConstraintLocalSpaceOffset(
    AdvancedSettings=(RotationOrderForFilter=XZY)
    Child=(Type=Bone,Name="bb_main")
    Filter=(bX=True,bY=True,bZ=True)
    Parents=((Item=(Type=Bone,Name="spine_02"),Weight=0.5))
    Weight=1
   )
   > &RigUnit_ModifyTransforms(
    ItemToModify=((Item=(Type=Bone,Name="bb_main"),Transform=(Rotation=(X=0,Y=0,Z=0,W=1),Translation=(X=-1,Y=2,Z=0),Scale3D=(X=1,Y=1,Z=1))))
    Weight=1
    ItemToModify.0.Transform.Rotation=&RigVMFunction_MathQuaternionFromEuler(Euler=(X=0,Y=0,Z=-20)).Result
   )
   RigUnit_GetTransform = &RigUnit_GetTransform(Item=(Type=Control,Name="foot_IK_l"))
   > /Script/ControlRig.RigUnit_TwoBoneIKSimplePerItem(
    DebugSettings=(Scale=10,WorldOffset=(Rotation=(X=0,Y=0,Z=0,W=1),Translation=(X=0,Y=0,Z=0),Scale3D=(X=1,Y=1,Z=1)))
    EffectorItem=(Type=Bone,Name="bb_foot_l")
    ItemA=(Type=Bone,Name="bb_thigh_l")
    ItemB=(Type=Bone,Name="bb_calf_l")
    PoleVectorSpace=(Type=Bone)
    PrimaryAxis=(X=-1,Y=0,Z=0)
    SecondaryAxis=(X=0,Y=1,Z=0)
    SecondaryAxisWeight=1
    StretchMaximumRatio=1.25
    StretchStartRatio=1
    Weight=1
    bEnableStretch=true
    bPropagateToChildren=True
    Effector=@RigUnit_GetTransform.Transform
    PoleVector=&RigUnit_GetTransform(Item=(Type=Control,Name="PV_l")).Transform.Translation
   ) {is=struct RigVMUnitNode}
   > &RigUnit_TwoBoneIKSimplePerItem(
    DebugSettings=(Scale=10,WorldOffset=(Rotation=(X=0,Y=0,Z=0,W=1),Translation=(X=0,Y=0,Z=0),Scale3D=(X=1,Y=1,Z=1)))
    EffectorItem=(Type=Bone,Name="bb_foot_r")
    ItemA=(Type=Bone,Name="bb_thigh_r")
    ItemB=(Type=Bone,Name="bb_calf_r")
    PoleVectorSpace=(Type=Bone)
    PrimaryAxis=(X=1,Y=0,Z=0)
    SecondaryAxis=(X=0,Y=-1,Z=0)
    SecondaryAxisWeight=1
    StretchMaximumRatio=1.25
    StretchStartRatio=1
    Weight=1
    bEnableStretch=true
    bPropagateToChildren=True
    Effector=&RigUnit_GetTransform(Item=(Type=Control,Name="foot_IK_r")).Transform
    PoleVector=&RigUnit_GetTransform(Item=(Type=Control,Name="PV_r")).Transform.Translation
   )
   > &RigUnit_ModifyTransforms(
    ItemToModify=((Item=(Type=Control,Name="foot_IK_l"),Transform=(Rotation=(X=0,Y=0,Z=0,W=1),Translation=(X=0,Y=0,Z=0),Scale3D=(X=1,Y=1,Z=1))))
    Weight=1
   )
   &RigUnit_GetRelativeTransformForItem(Child=(Type=Control,Name="foot_IK_l"), Parent=(Type=Bone,Name="pelvis"))
  ]}
]
hierarchy [
 root {loc=<0 0 0>, rot=<0 0 0 1>}
 pelvis {parent=root, loc=<0 -0.7321 92.4218>, rot=<0.0127 0.707 -0.0127 -0.707>}
 bb_main {parent=pelvis, loc=<11.2803 -5.1445 0>, rot=<0 0 0 1>}
 bb_spine_01 {parent=bb_main, loc=<0 0 0>, rot=<0 0 0.0532 -0.9986>}
 bb_spine_02 {parent=bb_spine_01, loc=<4.5152 0.7144 0>, rot=<0 0 0.0014 -1>}
 bb_spine_03 {parent=bb_spine_02, loc=<4.6106 0.1659 0>, rot=<0 0 0.0012 -1>}
 bb_clavicle_l {parent=bb_spine_03, loc=<4.0248 -0.0732 -1.7641>, rot=<0.0433 -0.7292 -0.0112 -0.6828>}
 bb_upperarm_l {parent=bb_clavicle_l, loc=<4.7729 -0.5778 0.2961>, rot=<0.0086 -0.3668 -0.0186 -0.9301>}
 bb_lowerarm_l {parent=bb_upperarm_l, loc=<7.3323 0 0>, rot=<0 0 -0.0748 -0.9972>}
 bb_hand_l {parent=bb_lowerarm_l, loc=<5.5785 0 0>, rot=<-0.6937 -0.0003 0.0003 -0.7203>}
 bb_thumb_01_l {parent=bb_hand_l, loc=<0.0119 -0.8352 -1.879>, rot=<0.5158 -0.2585 0.1074 -0.8097>}
 bb_thumb_02_l {parent=bb_thumb_01_l, loc=<1.4683 0 0>, rot=<0 0 0.0067 -1>}
 bb_middle_01_l {parent=bb_hand_l, loc=<1.9968 -0.3856 0.3029>, rot=<-0.1222 -0.0318 0.3326 -0.9346>}
 bb_middle_02_l {parent=bb_middle_01_l, loc=<1.1707 0 0>, rot=<0 0 0 1>}
 bb_index_01_l {parent=bb_hand_l, loc=<1.8862 -0.1693 -1.145>, rot=<-0.0165 -0.0143 0.2798 -0.9598>}
 bb_index_02_l {parent=bb_index_01_l, loc=<0.9331 0 0>, rot=<0 0 0.0474 -0.9989>}
 bb_ring_01_l {parent=bb_hand_l, loc=<1.5521 -1.0275 1.4824>, rot=<-0.2099 0.0148 0.2833 -0.9357>}
 bb_ring_02_l {parent=bb_ring_01_l, loc=<1.0828 0 0>, rot=<0 0 0 1>}
 bb_head {parent=bb_spine_03, loc=<6.9175 0.7634 0>, rot=<0 0 -0.0337 -0.9994>}
 bb_eyeMain_l {parent=bb_head, loc=<9.5786 2.0676 -3.4409>, rot=<0.2331 -0.6676 -0.2331 -0.6676>}
 bb_bb_eyeTp_l {parent=bb_eyeMain_l, loc=<0 0 0>, rot=<-0.3258 0 0 -0.9455>}
 bb_bb_eyeBt_l {parent=bb_eyeMain_l, loc=<0 0 0>, rot=<-0.7996 0 0 -0.6006>}
 bb_bb_pupil_l {parent=bb_eyeMain_l, loc=<0 2.2038 0.7841>, rot=<-0.3258 0 0 -0.9455>}
 bb_eyeMain_r {parent=bb_head, loc=<9.5786 2.0676 3.4409>, rot=<0.7071 0.0029 -0.7071 0.0029>}
 bb_bb_eyeTp_r {parent=bb_eyeMain_r, loc=<0 0 0>, rot=<0 0 0 1>}
 bb_bb_eyeBt_r {parent=bb_eyeMain_r, loc=<0 0 0>, rot=<0 0 0 1>}
 bb_bb_pupil_r {parent=bb_eyeMain_r, loc=<0 -2.2191 0.7398>, rot=<0 0 0 1>}
 bb_ear_l {parent=bb_head, loc=<13.5164 2.1099 -5.6287>, rot=<0.0213 -0.5039 0.006 -0.8635>}
 bb_cheek_l {parent=bb_head, loc=<5.3425 1.3217 -4.7764>, rot=<0.0187 -0.7302 0.0118 -0.6828>}
 bb_ear_r {parent=bb_head, loc=<13.5167 2.1099 5.6287>, rot=<-0.5039 -0.0213 0.8635 0.006>}
 bb_cheek_r {parent=bb_head, loc=<5.342 1.3217 4.7764>, rot=<0.7302 0.0187 -0.6828 -0.0118>}
 bb_browOt_r {parent=bb_head, loc=<11.1632 5.437 4.7904>, rot=<0.7071 0.0029 -0.7071 0.0029>}
 bb_browIn_r {parent=bb_head, loc=<10.9694 6.5883 2.085>, rot=<0.7071 0.0029 -0.7071 0.0029>}
 bb_browOt_l {parent=bb_head, loc=<11.1632 5.437 -4.7904>, rot=<0.0029 -0.7071 -0.0029 -0.7071>}
 bb_browIn_l {parent=bb_head, loc=<10.9694 6.5883 -2.085>, rot=<0.0029 -0.7071 -0.0029 -0.7071>}
 bb_clavicle_r {parent=bb_spine_03, loc=<4.0252 -0.0731 1.7641>, rot=<0.7292 0.0433 -0.6828 0.0112>}
 bb_upperarm_r {parent=bb_clavicle_r, loc=<-4.773 0.5778 -0.2955>, rot=<0.0081 -0.3668 -0.0184 -0.9301>}
 bb_lowerarm_r {parent=bb_upperarm_r, loc=<-7.3317 0 -0.0004>, rot=<0 0 -0.0661 -0.9978>}
 bb_hand_r {parent=bb_lowerarm_r, loc=<-5.579 0.0001 0.0005>, rot=<-0.6933 -0.0003 0.0003 -0.7206>}
 bb_thumb_01_r {parent=bb_hand_r, loc=<-0.0118 0.8351 1.879>, rot=<0.5158 -0.2585 0.1074 -0.8097>}
 bb_thumb_02_r {parent=bb_thumb_01_r, loc=<-1.4683 0 0>, rot=<0 0 0.0067 -1>}
 bb_middle_01_r {parent=bb_hand_r, loc=<-1.9968 0.3856 -0.3029>, rot=<-0.1222 -0.0318 0.3326 -0.9346>}
 bb_middle_02_r {parent=bb_middle_01_r, loc=<-1.1702 -0.0001 0.0001>, rot=<0 0 0 -1>}
 bb_index_01_r {parent=bb_hand_r, loc=<-1.8864 0.1695 1.145>, rot=<-0.0165 -0.0143 0.2798 -0.9598>}
 bb_index_02_r {parent=bb_index_01_r, loc=<-0.9331 -0.0001 0>, rot=<0 0 0.0474 -0.9989>}
 bb_ring_01_r {parent=bb_hand_r, loc=<-1.5518 1.0272 -1.4824>, rot=<-0.2099 0.0148 0.2833 -0.9357>}
 bb_ring_02_r {parent=bb_ring_01_r, loc=<-1.0829 0 0>, rot=<0 0 0 1>}
 bb_pelvis {parent=bb_main, loc=<0 0 0>, rot=<0 0 0 1>}
 bb_thigh_l {parent=bb_pelvis, loc=<-1.8815 0.8743 -3.3495>, rot=<0 0 0 1>}
 bb_calf_l {parent=bb_thigh_l, loc=<-5.1 0 0>, rot=<0 0 -0.0203 -0.9998>}
 bb_foot_l {parent=bb_calf_l, loc=<-5.1 0 0>, rot=<0 0 0 1>}
 bb_ball_l {parent=bb_foot_l, loc=<-1.0413 2.9601 -0.0321>, rot=<0 0 0 1>}
 bb_thigh_r {parent=bb_pelvis, loc=<-1.8811 0.8742 3.3495>, rot=<0 0 1 0>}
 bb_calf_r {parent=bb_thigh_r, loc=<5.1005 0 0>, rot=<0 0 -0.0203 -0.9998>}
 bb_foot_r {parent=bb_calf_r, loc=<5.1 0 0>, rot=<0 0 0 1>}
 bb_ball_r {parent=bb_foot_r, loc=<1.0413 -2.9601 0.0321>, rot=<0 0 0 1>}
 bb_tail {parent=bb_pelvis, loc=<1.2034 -5.8756 0>, rot=<0 0 0.7071 -0.7071>}
 Controller {parent=bb_main, loc=<6.3489 19.1172 0>, rot=<-0.0127 -0.707 0.0127 -0.707>}
 bb_mainAim {parent=bb_main, loc=<0.9528 26.5807 0>, rot=<0 0 0 1>}
 spine_01 {parent=pelvis, loc=<5.6002 -0.5128 0>, rot=<0 0 0.0532 -0.9986>}
 spine_02 {parent=spine_01, loc=<8.0718 -0.0114 0>, rot=<0 0 0.0014 -1>}
 spine_03 {parent=spine_02, loc=<7.0362 -0.0083 0>, rot=<0 0 0.0012 -1>}
 spine_04 {parent=spine_03, loc=<8.0992 -0.0104 0>, rot=<0 0 0.0013 -1>}
 spine_05 {parent=spine_04, loc=<10.0112 -0.015 0>, rot=<0 0 0.001 -1>}
 clavicle_l {parent=spine_05, loc=<2.5856 -0.2866 -4.3777>, rot=<0.0449 -0.7291 -0.0127 -0.6828>}
 upperarm_l {parent=clavicle_l, loc=<18.2269 0 0>, rot=<0.0231 -0.2805 0.0014 -0.9596>}
 lowerarm_l {parent=upperarm_l, loc=<29.0284 0 0>, rot=<0 0 -0.2475 -0.9689>}
 hand_l {parent=lowerarm_l, loc=<25.8757 0 0>, rot=<-0.6913 -0.0958 0.1331 -0.7037>}
 weapon_l {parent=hand_l, loc=<-0.1974 -3.4111 0.3811>, rot=<0 0 0 1>}
 thumb_01_l {parent=hand_l, loc=<1.3118 -0.3191 -2.459>, rot=<0.6546 -0.4042 0.0664 -0.6354>}
 thumb_02_l {parent=thumb_01_l, loc=<5.9336 -0.0448 0.0401>, rot=<-0.0088 0.005 0.2979 -0.9545>}
 thumb_03_l {parent=thumb_02_l, loc=<2.2223 0.3682 0>, rot=<0 0 0.0067 -1>}
 index_metacarpal_l {parent=hand_l, loc=<1.6535 0.0183 -2.96>, rot=<0.0158 -0.0124 -0.0173 -0.9996>}
 index_01_l {parent=index_metacarpal_l, loc=<8.3008 0.0357 -0.0524>, rot=<-0.0044 0.0485 0.0924 -0.9945>}
 index_02_l {parent=index_01_l, loc=<3.3587 -0.3004 0>, rot=<0 0 0.3225 -0.9466>}
 index_03_l {parent=index_02_l, loc=<2.5578 -0.123 0>, rot=<0 0 0.2559 -0.9667>}
 middle_metacarpal_l {parent=hand_l, loc=<2.1856 -0.1725 -0.5984>, rot=<0.006 0.0168 -0.0329 -0.9993>}
 middle_01_l {parent=middle_metacarpal_l, loc=<7.3275 -0.2447 0.0129>, rot=<-0.0208 0.0042 0.0858 -0.9961>}
 middle_02_l {parent=middle_01_l, loc=<4.016 0.1415 0>, rot=<0 0 0.3519 -0.936>}
 middle_03_l {parent=middle_02_l, loc=<2.6103 0.0517 0>, rot=<0 0 0.1707 -0.9853>}
 ring_metacarpal_l {parent=hand_l, loc=<2.2074 -0.3625 1.3638>, rot=<-0.0284 0.0722 -0.0458 -0.9959>}
 ring_01_l {parent=ring_metacarpal_l, loc=<7.0276 -0.6134 -0.2041>, rot=<-0.0484 -0.0225 0.1894 -0.9805>}
 ring_02_l {parent=ring_01_l, loc=<3.6174 0.4647 0>, rot=<0 0 0.2806 -0.9598>}
 ring_03_l {parent=ring_02_l, loc=<2.4152 0.114 0>, rot=<0 0 0.1189 -0.9929>}
 pinky_metacarpal_l {parent=hand_l, loc=<1.7616 -0.9753 3.3814>, rot=<-0.0397 0.0814 -0.069 -0.9935>}
 pinky_01_l {parent=pinky_metacarpal_l, loc=<7.1709 -0.4831 -0.1094>, rot=<-0.0676 -0.0132 0.2236 -0.9723>}
 pinky_02_l {parent=pinky_01_l, loc=<2.8274 0.0893 0>, rot=<0 0 0.2854 -0.9584>}
 pinky_03_l {parent=pinky_02_l, loc=<2.2046 0.0307 0>, rot=<0 0 0.1445 -0.9895>}
 lowerarm_twist_02_l {parent=lowerarm_l, loc=<9.6199 0 0>, rot=<0 0 0 1>}
 lowerarm_twist_01_l {parent=lowerarm_l, loc=<17.8549 0 0>, rot=<0 0 0 1>}
 upperarm_twist_01_l {parent=upperarm_l, loc=<0.0041 0 0>, rot=<0 0 0 1>}
 upperarm_twist_02_l {parent=upperarm_l, loc=<13.7941 0 0>, rot=<0 0 0 1>}
 neck_01 {parent=spine_05, loc=<10.815 1.9196 0>, rot=<0 0 -0.1864 -0.9825>}
 neck_02 {parent=neck_01, loc=<4.7347 0.3045 0>, rot=<0 0 0.0363 -0.9993>}
 head {parent=neck_02, loc=<5.7632 0.5552 0>, rot=<0 0 0.115 -0.9934>}
 clavicle_r {parent=spine_05, loc=<2.5855 -0.2866 4.3777>, rot=<0.7291 0.0449 -0.6828 0.0127>}
 upperarm_r {parent=clavicle_r, loc=<-18.259 0.938 -0.1214>, rot=<0.0231 -0.2805 0.0014 -0.9596>}
 lowerarm_r {parent=upperarm_r, loc=<-29.0284 0 0>, rot=<0 0 -0.2475 -0.9689>}
 hand_r {parent=lowerarm_r, loc=<-25.8757 0 0>, rot=<-0.6913 -0.0958 0.1331 -0.7037>}
 weapon_r {parent=hand_r, loc=<0.1974 3.4111 -0.3811>, rot=<0 0 0 1>}
 thumb_01_r {parent=hand_r, loc=<-1.3118 0.3191 2.459>, rot=<0.6546 -0.4042 0.0664 -0.6354>}
 thumb_02_r {parent=thumb_01_r, loc=<-5.9336 0.0448 -0.0401>, rot=<-0.0088 0.005 0.2979 -0.9545>}
 thumb_03_r {parent=thumb_02_r, loc=<-2.2223 -0.3682 0>, rot=<0 0 0.0067 -1>}
 index_metacarpal_r {parent=hand_r, loc=<-1.6535 -0.0183 2.96>, rot=<0.0158 -0.0124 -0.0173 -0.9996>}
 index_01_r {parent=index_metacarpal_r, loc=<-8.3008 -0.0357 0.0524>, rot=<-0.0044 0.0485 0.0924 -0.9945>}
 index_02_r {parent=index_01_r, loc=<-3.3587 0.3004 0>, rot=<0 0 0.3225 -0.9466>}
 index_03_r {parent=index_02_r, loc=<-2.5578 0.123 0>, rot=<0 0 0.2559 -0.9667>}
 middle_metacarpal_r {parent=hand_r, loc=<-2.1856 0.1725 0.5984>, rot=<0.006 0.0168 -0.0329 -0.9993>}
 middle_01_r {parent=middle_metacarpal_r, loc=<-7.3275 0.2447 -0.0129>, rot=<-0.0208 0.0042 0.0858 -0.9961>}
 middle_02_r {parent=middle_01_r, loc=<-4.016 -0.1415 0>, rot=<0 0 0.3519 -0.936>}
 middle_03_r {parent=middle_02_r, loc=<-2.6103 -0.0517 0>, rot=<0 0 0.1707 -0.9853>}
 ring_metacarpal_r {parent=hand_r, loc=<-2.2074 0.3625 -1.3638>, rot=<-0.0284 0.0722 -0.0458 -0.9959>}
 ring_01_r {parent=ring_metacarpal_r, loc=<-7.0276 0.6134 0.2041>, rot=<-0.0484 -0.0225 0.1894 -0.9805>}
 ring_02_r {parent=ring_01_r, loc=<-3.6174 -0.4647 0>, rot=<0 0 0.2806 -0.9598>}
 ring_03_r {parent=ring_02_r, loc=<-2.4152 -0.114 0>, rot=<0 0 0.1189 -0.9929>}
 pinky_metacarpal_r {parent=hand_r, loc=<-1.7616 0.9753 -3.3814>, rot=<-0.0397 0.0814 -0.069 -0.9935>}
 pinky_01_r {parent=pinky_metacarpal_r, loc=<-7.1709 0.4831 0.1094>, rot=<-0.0676 -0.0132 0.2236 -0.9723>}
 pinky_02_r {parent=pinky_01_r, loc=<-2.8274 -0.0893 0>, rot=<0 0 0.2854 -0.9584>}
 pinky_03_r {parent=pinky_02_r, loc=<-2.2046 -0.0307 0>, rot=<0 0 0.1445 -0.9895>}
 lowerarm_twist_02_r {parent=lowerarm_r, loc=<-9.6199 0 0>, rot=<0 0 0 -1>}
 lowerarm_twist_01_r {parent=lowerarm_r, loc=<-17.8549 0 0>, rot=<0 0 0 -1>}
 upperarm_twist_01_r {parent=upperarm_r, loc=<-0.0041 0 0>, rot=<0 0 0 -1>}
 upperarm_twist_02_r {parent=upperarm_r, loc=<-13.7941 0 0>, rot=<0 0 0 -1>}
 foot_IK_l {controlType=EULER_TRANSFORM, parent=pelvis, type=Control, loc=<8 0 93>, rot=<-0.0923 -0.7011 -0.2706 0.6533>}
 PV_l {controlType=EULER_TRANSFORM, parent=pelvis, type=Control, loc=<10 10 -20>, rot=<0 0 -0.1056 0.9944>}
 foot_IK_r {controlType=EULER_TRANSFORM, parent=pelvis, type=Control, loc=<1 0 7>, rot=<0 0 0.9944 0.1056>}
 PV_r {controlType=EULER_TRANSFORM, parent=pelvis, type=Control, loc=<15 20 8>, rot=<0 0 -0.1056 0.9944>}
]
