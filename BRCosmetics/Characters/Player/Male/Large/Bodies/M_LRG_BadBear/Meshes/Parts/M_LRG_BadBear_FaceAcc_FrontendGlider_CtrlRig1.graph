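# Control Rig for the Bad Bear character's frontend glider FaceAcc part, adjusting the small "bb_" bear skeleton parented to the pelvis.
# Two spherical pose readers measure spine_01's orientation against root at different angular offsets.
# Each reader weights a transform on bb_main: the first rotates it, the second counter-rotates and shifts it.
# A rotation constraint from spine_01 to bb_spine_01 exists but is disconnected from execution.
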ControlRigBlueprint /BRCosmetics/Characters/Player/Male/Large/Bodies/M_LRG_BadBear/Meshes/Parts/M_LRG_BadBear_FaceAcc_FrontendGlider_CtrlRig1
graphs [
 RigVMFunctionLibrary {links [], nodes []}
 RigVMModel {nodes [
   RigUnit_BeginExecution = &RigUnit_BeginExecution()
   n1 = &RigUnit_SphericalPoseReader(
    ActiveRegionScaleFactors=(PositiveWidth=1,NegativeWidth=1,PositiveHeight=1,NegativeHeight=1)
    ActiveRegionSize=0.1
    Debug=(bDrawDebug=false,DebugScale=25,DebugSegments=20,DebugThickness=0.25)
    DriverAxis=(X=1,Y=0,Z=0)
    DriverItem=(Type=Bone,Name="spine_01")
    FalloffRegionScaleFactors=(PositiveWidth=0.1,NegativeWidth=0.1,PositiveHeight=0.1,NegativeHeight=0.1)
    FalloffSize=0.2
    OptionalParentItem=(Type=Bone,Name="root")
    RotationOffset=(X=90,Y=0,Z=-70)
    ExecutePin=@RigUnit_BeginExecution.ExecutePin
   )
   > &RigUnit_ModifyTransforms(
    ItemToModify=((Item=(Type=Bone,Name="bb_main"),Transform=(Rotation=(X=0,Y=0,Z=0,W=1),Translation=(X=0,Y=0,Z=0),Scale3D=(X=1,Y=1,Z=1))))
    ItemToModify.0.Transform.Rotation=&RigVMFunction_MathQuaternionFromEuler(Euler=(X=0,Y=0,Z=-10)).Result
    Weight=@n1.OutputParam
   )
   > n2 = &RigUnit_SphericalPoseReader(
    ActiveRegionScaleFactors=(PositiveWidth=0.5,NegativeWidth=0.5,PositiveHeight=0.5,NegativeHeight=0.5)
    ActiveRegionSize=0.1
    Debug=(bDrawDebug=false,DebugScale=25,DebugSegments=20,DebugThickness=0.25)
    DriverAxis=(X=1,Y=0,Z=0)
    DriverItem=(Type=Bone,Name="spine_01")
    FalloffRegionScaleFactors=(PositiveWidth=0.1,NegativeWidth=0.1,PositiveHeight=0.1,NegativeHeight=0.1)
    FalloffSize=0.2
    OptionalParentItem=(Type=Bone,Name="root")
    RotationOffset=(X=90,Y=0,Z=-100)
   )
   > &RigUnit_ModifyTransforms(
    ItemToModify=((Item=(Type=Bone,Name="bb_main"),Transform=(Rotation=(X=0,Y=0,Z=0,W=1),Translation=(X=0,Y=-6,Z=0),Scale3D=(X=1,Y=1,Z=1))))
    ItemToModify.0.Transform.Rotation=&RigVMFunction_MathQuaternionFromEuler(Euler=(X=0,Y=0,Z=20)).Result
    Weight=@n2.OutputParam
   )
   &RigUnit_RotationConstraintLocalSpaceOffset(
    AdvancedSettings=(RotationOrderForFilter=XZY)
    Child=(Type=Bone,Name="bb_spine_01")
    ExecutePin=()
    Filter=(bX=True,bY=True,bZ=True)
    Parents=((Item=(Type=Bone,Name="spine_01"),Weight=1))
    Weight=1
   )
  ]}
]
hierarchy [
 root {loc=<0 0 0>, rot=<0 0 0 1>}
 pelvis {parent=root, loc=<0 -0.7321 92.4218>, rot=<0.0127 0.707 -0.0127 -0.707>}
 spine_01 {parent=pelvis, loc=<5.6002 -0.5128 0>, rot=<0 0 0.0532 -0.9986>}
 spine_02 {parent=spine_01, loc=<8.0718 -0.0114 0>, rot=<0 0 0.0014 -1>}
 spine_03 {parent=spine_02, loc=<7.0362 -0.0083 0>, rot=<0 0 0.0012 -1>}
 spine_04 {parent=spine_03, loc=<8.0992 -0.0104 0>, rot=<0 0 0.0013 -1>}
 spine_05 {parent=spine_04, loc=<10.0112 -0.015 0>, rot=<0 0 0.001 -1>}
 clavicle_l {parent=spine_05, loc=<2.5856 -0.2866 -4.3777>, rot=<0.0449 -0.7291 -0.0127 -0.6828>}
 upperarm_l {parent=clavicle_l, loc=<18.2269 0 0>, rot=<0.0231 -0.2805 0.0014 -0.9596>}
 lowerarm_l {parent=upperarm_l, loc=<29.0284 0 0>, rot=<0 0 -0.2475 -0.9689>}
 hand_l {parent=lowerarm_l, loc=<25.8757 0 0>, rot=<-0.6913 -0.0958 0.1331 -0.7037>}
 neck_01 {parent=spine_05, loc=<10.815 1.9196 0>, rot=<0 0 -0.1864 -0.9825>}
 neck_02 {parent=neck_01, loc=<4.7347 0.3045 0>, rot=<0 0 0.0363 -0.9993>}
 head {parent=neck_02, loc=<5.7632 0.5552 0>, rot=<0 0 0.115 -0.9934>}
 clavicle_r {parent=spine_05, loc=<2.5855 -0.2866 4.3777>, rot=<0.7291 0.0449 -0.6828 0.0127>}
 upperarm_r {parent=clavicle_r, loc=<-18.259 0.938 -0.1214>, rot=<0.0231 -0.2805 0.0014 -0.9596>}
 lowerarm_r {parent=upperarm_r, loc=<-29.0284 0 0>, rot=<0 0 -0.2475 -0.9689>}
 hand_r {parent=lowerarm_r, loc=<-25.8757 0 0>, rot=<-0.6913 -0.0958 0.1331 -0.7037>}
 bb_main {parent=pelvis, loc=<11.2803 -5.1445 0>, rot=<0 0 0 1>}
 bb_spine_01 {parent=bb_main, loc=<0 0 0>, rot=<0 0 0.0532 -0.9986>}
 bb_spine_02 {parent=bb_spine_01, loc=<4.5152 0.7144 0>, rot=<0 0 0.0014 -1>}
 bb_spine_03 {parent=bb_spine_02, loc=<4.6106 0.1659 0>, rot=<0 0 0.0012 -1>}
 bb_clavicle_l {parent=bb_spine_03, loc=<4.0248 -0.0732 -1.7641>, rot=<0.0433 -0.7292 -0.0112 -0.6828>}
 bb_upperarm_l {parent=bb_clavicle_l, loc=<4.7729 -0.5778 0.2961>, rot=<0.0086 -0.3668 -0.0186 -0.9301>}
 bb_lowerarm_l {parent=bb_upperarm_l, loc=<7.3323 0 0>, rot=<0 0 -0.0748 -0.9972>}
 bb_hand_l {parent=bb_lowerarm_l, loc=<5.5785 0 0>, rot=<-0.6937 -0.0003 0.0003 -0.7203>}
 bb_thumb_01_l {parent=bb_hand_l, loc=<0.0119 -0.8352 -1.879>, rot=<0.5158 -0.2585 0.1074 -0.8097>}
 bb_thumb_02_l {parent=bb_thumb_01_l, loc=<1.4683 0 0>, rot=<0 0 0.0067 -1>}
 bb_middle_01_l {parent=bb_hand_l, loc=<1.9968 -0.3856 0.3029>, rot=<-0.1222 -0.0318 0.3326 -0.9346>}
 bb_middle_02_l {parent=bb_middle_01_l, loc=<1.1707 0 0>, rot=<0 0 0 1>}
 bb_index_01_l {parent=bb_hand_l, loc=<1.8862 -0.1693 -1.145>, rot=<-0.0165 -0.0143 0.2798 -0.9598>}
 bb_index_02_l {parent=bb_index_01_l, loc=<0.9331 0 0>, rot=<0 0 0.0474 -0.9989>}
 bb_ring_01_l {parent=bb_hand_l, loc=<1.5521 -1.0275 1.4824>, rot=<-0.2099 0.0148 0.2833 -0.9357>}
 bb_ring_02_l {parent=bb_ring_01_l, loc=<1.0828 0 0>, rot=<0 0 0 1>}
 bb_head {parent=bb_spine_03, loc=<6.9175 0.7634 0>, rot=<0 0 -0.0337 -0.9994>}
 bb_eyeMain_l {parent=bb_head, loc=<9.5786 2.0676 -3.4409>, rot=<0.2331 -0.6676 -0.2331 -0.6676>}
 bb_bb_eyeTp_l {parent=bb_eyeMain_l, loc=<0 0 0>, rot=<-0.3258 0 0 -0.9455>}
 bb_bb_eyeBt_l {parent=bb_eyeMain_l, loc=<0 0 0>, rot=<-0.7996 0 0 -0.6006>}
 bb_bb_pupil_l {parent=bb_eyeMain_l, loc=<0 2.2038 0.7841>, rot=<-0.3258 0 0 -0.9455>}
 bb_eyeMain_r {parent=bb_head, loc=<9.5786 2.0676 3.4409>, rot=<0.7071 0.0029 -0.7071 0.0029>}
 bb_bb_eyeTp_r {parent=bb_eyeMain_r, loc=<0 0 0>, rot=<0 0 0 1>}
 bb_bb_eyeBt_r {parent=bb_eyeMain_r, loc=<0 0 0>, rot=<0 0 0 1>}
 bb_bb_pupil_r {parent=bb_eyeMain_r, loc=<0 -2.2191 0.7398>, rot=<0 0 0 1>}
 bb_ear_l {parent=bb_head, loc=<13.5164 2.1099 -5.6287>, rot=<0.0213 -0.5039 0.006 -0.8635>}
 bb_cheek_l {parent=bb_head, loc=<5.3425 1.3217 -4.7764>, rot=<0.0187 -0.7302 0.0118 -0.6828>}
 bb_ear_r {parent=bb_head, loc=<13.5167 2.1099 5.6287>, rot=<-0.5039 -0.0213 0.8635 0.006>}
 bb_cheek_r {parent=bb_head, loc=<5.342 1.3217 4.7764>, rot=<0.7302 0.0187 -0.6828 -0.0118>}
 bb_browOt_r {parent=bb_head, loc=<11.1632 5.437 4.7904>, rot=<0.7071 0.0029 -0.7071 0.0029>}
 bb_browIn_r {parent=bb_head, loc=<10.9694 6.5883 2.085>, rot=<0.7071 0.0029 -0.7071 0.0029>}
 bb_browOt_l {parent=bb_head, loc=<11.1632 5.437 -4.7904>, rot=<0.0029 -0.7071 -0.0029 -0.7071>}
 bb_browIn_l {parent=bb_head, loc=<10.9694 6.5883 -2.085>, rot=<0.0029 -0.7071 -0.0029 -0.7071>}
 bb_clavicle_r {parent=bb_spine_03, loc=<4.0252 -0.0731 1.7641>, rot=<0.7292 0.0433 -0.6828 0.0112>}
 bb_upperarm_r {parent=bb_clavicle_r, loc=<-4.773 0.5778 -0.2955>, rot=<0.0081 -0.3668 -0.0184 -0.9301>}
 bb_lowerarm_r {parent=bb_upperarm_r, loc=<-7.3317 0 -0.0004>, rot=<0 0 -0.0661 -0.9978>}
 bb_hand_r {parent=bb_lowerarm_r, loc=<-5.579 0.0001 0.0005>, rot=<-0.6933 -0.0003 0.0003 -0.7206>}
 bb_thumb_01_r {parent=bb_hand_r, loc=<-0.0118 0.8351 1.879>, rot=<0.5158 -0.2585 0.1074 -0.8097>}
 bb_thumb_02_r {parent=bb_thumb_01_r, loc=<-1.4683 0 0>, rot=<0 0 0.0067 -1>}
 bb_middle_01_r {parent=bb_hand_r, loc=<-1.9968 0.3856 -0.3029>, rot=<-0.1222 -0.0318 0.3326 -0.9346>}
 bb_middle_02_r {parent=bb_middle_01_r, loc=<-1.1702 -0.0001 0.0001>, rot=<0 0 0 -1>}
 bb_index_01_r {parent=bb_hand_r, loc=<-1.8864 0.1695 1.145>, rot=<-0.0165 -0.0143 0.2798 -0.9598>}
 bb_index_02_r {parent=bb_index_01_r, loc=<-0.9331 -0.0001 0>, rot=<0 0 0.0474 -0.9989>}
 bb_ring_01_r {parent=bb_hand_r, loc=<-1.5518 1.0272 -1.4824>, rot=<-0.2099 0.0148 0.2833 -0.9357>}
 bb_ring_02_r {parent=bb_ring_01_r, loc=<-1.0829 0 0>, rot=<0 0 0 1>}
 bb_pelvis {parent=bb_main, loc=<0 0 0>, rot=<0 0 0 1>}
 bb_thigh_l {parent=bb_pelvis, loc=<-1.8815 0.8743 -3.3495>, rot=<0 0 0 1>}
 bb_calf_l {parent=bb_thigh_l, loc=<-5.1 0 0>, rot=<0 0 -0.0203 -0.9998>}
 bb_foot_l {parent=bb_calf_l, loc=<-5.1 0 0>, rot=<0 0 0 1>}
 bb_ball_l {parent=bb_foot_l, loc=<-1.0413 2.9601 -0.0321>, rot=<0 0 0 1>}
 bb_thigh_r {parent=bb_pelvis, loc=<-1.8811 0.8742 3.3495>, rot=<0 0 1 0>}
 bb_calf_r {parent=bb_thigh_r, loc=<5.1005 0 0>, rot=<0 0 -0.0203 -0.9998>}
 bb_foot_r {parent=bb_calf_r, loc=<5.1 0 0>, rot=<0 0 0 1>}
 bb_ball_r {parent=bb_foot_r, loc=<1.0413 -2.9601 0.0321>, rot=<0 0 0 1>}
 bb_tail {parent=bb_pelvis, loc=<1.2034 -5.8756 0>, rot=<0 0 0.7071 -0.7071>}
 Controller {parent=bb_main, loc=<6.3489 19.1172 0>, rot=<-0.0127 -0.707 0.0127 -0.707>}
 bb_mainAim {parent=bb_main, loc=<0.9528 26.5807 0>, rot=<0 0 0 1>}
 C_glabella_down_pose {type=Curve, loc=<0 0 0>, rot=<0 0 0 1>}
 C_glabella_up_pose {type=Curve, loc=<0 0 0>, rot=<0 0 0 1>}
 jaw_open_pose {type=Curve, loc=<0 0 0>, rot=<0 0 0 1>}
 L_upper_lip_raiser_pose {type=Curve, loc=<0 0 0>, rot=<0 0 0 1>}
 phoneme_mbp_pose {type=Curve, loc=<0 0 0>, rot=<0 0 0 1>}
 phoneme_oo_pose {type=Curve, loc=<0 0 0>, rot=<0 0 0 1>}
 R_blink_pose {type=Curve, loc=<0 0 0>, rot=<0 0 0 1>}
 R_brow_down_pose {type=Curve, loc=<0 0 0>, rot=<0 0 0 1>}
 R_brow_up_pose {type=Curve, loc=<0 0 0>, rot=<0 0 0 1>}
 R_Frown_pose {type=Curve, loc=<0 0 0>, rot=<0 0 0 1>}
 R_lip_corner_narrow_pose {type=Curve, loc=<0 0 0>, rot=<0 0 0 1>}
 R_smile_pose {type=Curve, loc=<0 0 0>, rot=<0 0 0 1>}
 R_squint_inner_pose {type=Curve, loc=<0 0 0>, rot=<0 0 0 1>}
 R_upper_lip_raiser_pose {type=Curve, loc=<0 0 0>, rot=<0 0 0 1>}
 BypassControlRig {type=Curve, loc=<0 0 0>, rot=<0 0 0 1>}
]
