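# Physics asset for the face accessory mesh part of the large male BadBear character body, covering the pelvis and tail bones.
# Each of the two bones has a single capsule body, a large one on bb_pelvis and a small one on bb_tail.
# One constraint joins the tail to the pelvis with narrow swing limits and a zero twist limit angle.
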PhysicsAsset /BRCosmetics/Characters/Player/Male/Large/Bodies/M_LRG_BadBear/Meshes/Parts/M_LRG_BadBear_FaceAcc_Physics
bodies {
 bb_pelvis [{shapeType=Capsule, length=12.5731, radius=8.7284, center=<-3.4817 1.7498 -1.5637>, rotation=<rot 0 63.0725 -73.8773>}]
 bb_tail [{shapeType=Capsule, length=2.3182, radius=1.9836, shapeName=bb_tail_capsule, center=<1 0 0>, rotation=<rot 90 0 0>}]
}
constraints [{bone1Name=bb_tail, bone2Name=bb_pelvis, swing1LimitDegrees=10, swing2LimitDegrees=5, twistLimitDegrees=0}]
