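# Animation Blueprint for the M_LRG_BadBear face accessory, animating a bear rig attached to the player.
# FrontEnd and InGame state machines (locomotion, jump, slide, swimming, skydive, DBNO) and slots layer from bb_main over the copied parent pose.
# The cached result gets spring, Control Rig, RigidBody and state-driven bone constraints and offsets.
# A front-end branch and bypass alpha can skip that chain, before final wing-glider constraints and Control Rig.
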AnimBlueprint /BRCosmetics/Characters/Player/Male/Large/Bodies/M_LRG_BadBear/Meshes/Parts/M_LRG_BadBear_FaceAcc_AnimBP
target_skeleton=&/M_LRG_BadBear_FaceAcc_Skeleton
graphs [
 AnimGraph {nodes [
   n1 = Get CharacterAimYaw_Idle() -> (CharacterAimYaw_Idle:double)
   n2 = Get bIsSprinting()
   n3 = Get CharacterAimYaw_Sprint() -> (CharacterAimYaw_Sprint:double)
   Apply Additive(Additive=Blendspace Player 'Head_look_BS'(X=SelectFloat(A=@n3, B=@n1, bPickA=@n2), Y=Get CharacterAimPitch() -> (CharacterAimPitch:float)) {blendSpace=/BRCosmetics/Characters/Player/Male/Large/Bodies/M_LRG_BadBear/Meshes/Parts/proto_anim/Head_look_BS}) {alphaBoolBlend {blendOption=Linear}}
   n4 = Get bIsParachuteOpened()
   "Transform (Modify) Bone - Bone: bb_main"!(ComponentPose, bAlphaBoolEnabled:bool=@n4, Translation="0,-8,5", Rotation="0,0,0", Scale="1,1,1") -> (Pose) {
    alphaInputType=Bool
    rotationMode=BMM_Additive
    translationMode=BMM_Additive
    alphaBoolBlend {blendOption=Linear}
    alphaScaleBiasClamp {bInterpResult=true, interpSpeedDecreasing=2, interpSpeedIncreasing=18}
    boneToModify {boneName=bb_main}
    translation {y=-8, z=5}
   }
   n5 = Use cached pose 'CachedPose_PreControlRig'()
   n6 = "Transform (Modify) Bone - Bone: bb_main"(
    ComponentPose=Spring controller - Bone: bb_main!(
     ComponentPose:Component Space Pose Link Structure=Local To Component(LocalPose=@n5)
     Alpha:float=Get SpringAlphaSpeedCap() -> (SpringAlphaSpeedCap:double)
    ) -> (Pose:Component Space Pose Link Structure) {is=title AnimGraphNode_SpringBone, maxDisplacement=5, springStiffness=85, alphaBoolBlend {blendOption=Linear}, springBone {boneName=bb_main}}
    bAlphaBoolEnabled=NOT Boolean(A=Get bIsInFrontEnd())
    Translation="1,4,2.5"
    Rotation="0,0,0"
    Scale="1,1,1"
   ) {
    alphaInputType=Bool
    rotationMode=BMM_Additive
    rotationSpace=BCS_BoneSpace
    scaleMode=BMM_Additive
    scaleSpace=BCS_BoneSpace
    translationMode=BMM_Additive
    alphaBoolBlend {blendOption=Linear}
    alphaScaleBiasClamp {bInterpResult=true, interpSpeedIncreasing=500}
    boneToModify {boneName=bb_main}
    translation=<1 4 2.5>
   }
   n7 = Control Rig(Source=Component To Local(ComponentPose=@n6), Alpha=Get ControlRigAlpha() -> (ControlRigAlpha:double)) {
    alphaBoolBlend {blendOption=Linear}
    controlRigAssetReference {blueprintRigClass=&/M_LRG_BadBear_FaceAcc_CtrlRig.M_LRG_BadBear_FaceAcc_CtrlRig_C}
   }
   n8 = RigidBody!(ComponentPose=Local To Component(LocalPose=@n7), Alpha=0.5, OverrideWorldGravity:Vector=Get GravityOverride(), ExternalForce) -> (Pose) {
    bClampLinearTranslationLimitToRefPose=true
    bOverrideWorldGravity=true
    overridePhysicsAsset=&/M_LRG_BadBear_FaceAcc_Physics
    simulationSpace=BaseBoneSpace
    alphaBoolBlend {blendOption=Linear}
    baseBoneRef {boneName=root}
    componentAppliedLinearAccClamp=<3000 3000 3000>
    componentLinearAccScale=<0.2 0.1 0.1>
    componentLinearVelScale=<0.2 0.1 0.1>
    overrideWorldGravity {z=-980}
    simSpaceSettings {
     maxAngularAcceleration=1000
     maxAngularVelocity=3
     maxLinearAcceleration=700
     maxLinearVelocity=500
     velocityScaleZ=0
     externalLinearDragV=<0.2 0.2>
    }
   }
   n9 = Get SurfaceSwimAmount() -> (SurfaceSwimAmount:double)
   n10 = "Constraint - bb_main (PT)"!(ComponentPose=@n8, Alpha=@n9, ConstraintWeights_0=1, ConstraintWeights_1:float=0) -> (Pose) {
    alphaBoolBlend {blendOption=Linear}
    boneToModify {boneName=bb_main}
    constraintSetup [{targetBone {boneName=spine_05}}, {transformType=Translation, targetBone {boneName=spine_03}}]
    constraintWeights [1,1]
   }
   n11 = "Transform (Modify) Bone - Bone: bb_main"!(ComponentPose=@n10, Alpha:float=@n9, Translation="0,10,7", Rotation="0,-20,0", Scale="1,1,1") -> (Pose) {
    rotationMode=BMM_Additive
    rotationSpace=BCS_BoneSpace
    scaleMode=BMM_Additive
    scaleSpace=BCS_BoneSpace
    translationMode=BMM_Additive
    alphaBoolBlend {blendOption=Linear}
    alphaScaleBiasClamp {bInterpResult=true}
    boneToModify {boneName=bb_main}
    rotation {yaw=-20}
    translation {y=10, z=7}
   }
   n12 = "Transform (Modify) Bone - Bone: bb_tail"(ComponentPose=@n11, Alpha=1, Translation="0,0,0", Rotation="0,5,0", Scale="1,1,1") {
    rotationMode=BMM_Additive
    rotationSpace=BCS_BoneSpace
    scaleMode=BMM_Additive
    scaleSpace=BCS_BoneSpace
    translationMode=BMM_Additive
    alphaBoolBlend {blendOption=Linear}
    alphaScaleBiasClamp {bInterpResult=true, interpSpeedIncreasing=500}
    boneToModify {boneName=bb_tail}
    rotation {yaw=5}
   }
   n13 = Get bIsSprinting()
   n14 = "Constraint - bb_main (RR)"!(ComponentPose=@n12, bAlphaBoolEnabled:bool=@n13, ConstraintWeights_0=0.5, ConstraintWeights_1:float=0.5) -> (Pose) {
    alphaInputType=Bool
    alphaBoolBlend {blendOption=Linear}
    boneToModify {boneName=bb_main}
    constraintSetup [{transformType=Rotation, targetBone {boneName=spine_05}}, {transformType=Rotation, targetBone {boneName=spine_03}}]
    constraintWeights [1,1]
   }
   n15 = "Transform (Modify) Bone - Bone: bb_main"!(ComponentPose=@n14, bAlphaBoolEnabled:bool=@n13, Translation="0,0,0", Rotation="0,-25,0", Scale="1,1,1") -> (Pose) {
    alphaInputType=Bool
    rotationMode=BMM_Additive
    rotationSpace=BCS_BoneSpace
    scaleMode=BMM_Additive
    scaleSpace=BCS_BoneSpace
    translationMode=BMM_Additive
    alphaBoolBlend {blendOption=Linear}
    boneToModify {boneName=bb_main}
    rotation {yaw=-25}
   }
   n16 = "Constraint - bb_main (RR)"(ComponentPose=@n15, bAlphaBoolEnabled=@n4, ConstraintWeights_0=1, ConstraintWeights_1=0) {
    alphaInputType=Bool
    alphaBoolBlend {blendOption=Linear}
    boneToModify {boneName=bb_main}
    constraintSetup [{transformType=Rotation, targetBone {boneName=pelvis}}, {transformType=Rotation, targetBone {boneName=spine_01}}]
    constraintWeights [1,1]
   }
   n17 = "Constraint - bb_spine_02 (RR)"!(ComponentPose=@n16, bAlphaBoolEnabled:bool=@n4, ConstraintWeights_0=0, ConstraintWeights_1:float=1) -> (Pose) {
    alphaInputType=Bool
    alphaBoolBlend {blendOption=Linear}
    boneToModify {boneName=bb_spine_02}
    constraintSetup [{transformType=Rotation, targetBone {boneName=pelvis}}, {transformType=Rotation, targetBone {boneName=spine_05}}]
    constraintWeights [1,1]
   }
   n18 = "Constraint - bb_head (RR)"!(ComponentPose=@n17, bAlphaBoolEnabled:bool=@n4, ConstraintWeights_0=1, ConstraintWeights_1:float=0) -> (Pose) {
    alphaInputType=Bool
    alphaBoolBlend {blendOption=Linear}
    boneToModify {boneName=bb_head}
    constraintSetup [{transformType=Rotation, targetBone {boneName=pelvis}}, {transformType=Rotation, targetBone {boneName=spine_05}}]
    constraintWeights [1,1]
   }
   n19 = "Transform (Modify) Bone - Bone: bb_main"(ComponentPose=@n18, bAlphaBoolEnabled=Get bIsSkydiveFloating(), Translation="0,0,0", Rotation="0,0,10", Scale="1,1,1") {
    alphaInputType=Bool
    rotationMode=BMM_Additive
    translationMode=BMM_Additive
    alphaBoolBlend {blendOption=Linear}
    alphaScaleBiasClamp {bInterpResult=true, interpSpeedDecreasing=2, interpSpeedIncreasing=18}
    boneToModify {boneName=bb_main}
    rotation {roll=10}
   }
   n20 = "Transform (Modify) Bone - Bone: bb_main"(ComponentPose=@n19, bAlphaBoolEnabled=Get bIsInFrontEnd(), Translation="0,0,0", Rotation="10,0.000,20", Scale="1,1,1") {
    alphaInputType=Bool
    rotationMode=BMM_Additive
    translationMode=BMM_Additive
    alphaBoolBlend {blendOption=Linear}
    alphaScaleBiasClamp {bInterpResult=true, interpSpeedDecreasing=2, interpSpeedIncreasing=18}
    boneToModify {boneName=bb_main}
    rotation {pitch=10, roll=20}
   }
   n21 = "Transform (Modify) Bone - Bone: bb_foot_l"!(ComponentPose=@n20, bAlphaBoolEnabled:bool=false, Translation="0,0,0", Rotation="9.850,1.758,30.152", Scale="1,1,1") -> (Pose) {
    alphaInputType=Bool
    rotationMode=BMM_Additive
    translationMode=BMM_Additive
    alphaBoolBlend {blendOption=Linear}
    alphaScaleBiasClamp {bInterpResult=true, interpSpeedDecreasing=2, interpSpeedIncreasing=18}
    boneToModify {boneName=bb_foot_l}
    rotation=<rot 9.85 1.758 30.152>
   }
   n22 = "Transform (Modify) Bone - Bone: bb_foot_r"!(ComponentPose=@n21, bAlphaBoolEnabled:bool=false, Translation="0,0,0", Rotation="-6.243,0.567,19.196", Scale="1,1,1") -> (Pose) {
    alphaInputType=Bool
    rotationMode=BMM_Additive
    translationMode=BMM_Additive
    alphaBoolBlend {blendOption=Linear}
    alphaScaleBiasClamp {bInterpResult=true, interpSpeedDecreasing=2, interpSpeedIncreasing=18}
    boneToModify {boneName=bb_foot_r}
    rotation=<rot -6.243 0.567 19.196>
   }
   n23 = Get frontEndNoEmote() -> (frontEndNoEmote:bool)
   n24 = "Transform (Modify) Bone - Bone: bb_thigh_l"!(
    ComponentPose=Local To Component(LocalPose=Use cached pose 'CachedPose_PreControlRig'())
    bAlphaBoolEnabled:bool=@n23
    Translation="-0.5,0,0"
    Rotation="0,10,0"
    Scale="1,1,1"
   ) -> (Pose) {
    alphaInputType=Bool
    rotationMode=BMM_Additive
    translationMode=BMM_Additive
    alphaBoolBlend {blendOption=Linear}
    alphaScaleBiasClamp {bInterpResult=true, interpSpeedDecreasing=2, interpSpeedIncreasing=18}
    boneToModify {boneName=bb_thigh_l}
    rotation {yaw=10}
    translation {x=-0.5}
   }
   n25 = "Transform (Modify) Bone - Bone: bb_thigh_r"!(ComponentPose=@n24, bAlphaBoolEnabled:bool=@n23, Translation="0.5,0,0", Rotation="0,-10,0", Scale="1,1,1") -> (Pose) {
    alphaInputType=Bool
    rotationMode=BMM_Additive
    translationMode=BMM_Additive
    alphaBoolBlend {blendOption=Linear}
    alphaScaleBiasClamp {bInterpResult=true, interpSpeedDecreasing=2, interpSpeedIncreasing=18}
    boneToModify {boneName=bb_thigh_r}
    rotation {yaw=-10}
    translation {x=0.5}
   }
   n26 = Blend Poses by bool(
    BlendPose_0=Component To Local(ComponentPose=@n25)
    BlendPose_1=Component To Local(ComponentPose=@n22)
    BlendTime_0=0.3
    BlendTime_1=0.3
    bActiveValue=Get FrontEndNoEmoteExceptCustomEmote() -> (FrontEndNoEmoteExceptCustomEmote:bool)
   ) {blendType=Linear}
   n27 = Use cached pose 'CachedPose_PreControlRig'()
   n28 = Get BypassControlRigAlpha() -> (BypassControlRigAlpha:float)
   n29 = Get IsWingGlider() -> (IsWingGlider:double)
   n30 = "Constraint - bb_main (RT)"!(
    ComponentPose=Local To Component(LocalPose=Two Way Blend(A=@n26, B=@n27, Alpha=@n28) {blendNode {alphaBoolBlend {blendOption=Linear}}})
    Alpha=@n29
    ConstraintWeights_0=1
    ConstraintWeights_1:float=1
   ) -> (Pose) {
    alphaBoolBlend {blendOption=Linear}
    boneToModify {boneName=bb_main}
    constraintSetup [{transformType=Rotation, targetBone {boneName=pelvis}}, {transformType=Translation, targetBone {boneName=pelvis}}]
    constraintWeights [1,1]
   }
   n31 = "Transform (Modify) Bone - Bone: bb_spine_01"(ComponentPose=@n30, Alpha=@n29, Translation="0,0,0", Rotation="0,0,30", Scale="1,1,1") {
    rotationMode=BMM_Additive
    translationMode=BMM_Additive
    alphaBoolBlend {blendOption=Linear}
    alphaScaleBiasClamp {bInterpResult=true, interpSpeedDecreasing=2, interpSpeedIncreasing=18}
    boneToModify {boneName=bb_spine_01}
    rotation {roll=30}
   }
   n32 = Control Rig(Source=Component To Local(ComponentPose=@n31), Alpha=@n29) {
    alphaBoolBlend {blendOption=Linear}
    controlRigAssetReference {blueprintRigClass=&/M_LRG_BadBear_FaceAcc_FrontendGlider_CtrlRig1.M_LRG_BadBear_FaceAcc_FrontendGlider_CtrlRig1_C}
   }
   n33 = "Constraint - Controller (TTTT)"!(
    ComponentPose=Local To Component(LocalPose=@n32)
    Alpha=@n29
    ConstraintWeights_0=1
    ConstraintWeights_1:float=1
    ConstraintWeights_2:float=1
    ConstraintWeights_3:float=1
   ) -> (Pose) {
    alphaBoolBlend {blendOption=Linear}
    boneToModify {boneName=Controller}
    constraintSetup [
     {offsetOption=None, transformType=Translation, targetBone {boneName=bb_hand_l}}
     {offsetOption=None, transformType=Translation, targetBone {boneName=bb_hand_r}}
     {transformType=Translation, targetBone {boneName=None}}
     {transformType=Translation, targetBone {boneName=None}}
    ]
    constraintWeights [1,1,1,1]
   }
   n34 = "Transform (Modify) Bone - Bone: Controller"(ComponentPose=@n33, Alpha=@n29, Translation="0,2,0", Rotation="0,0,0", Scale="1,1,1") {
    rotationMode=BMM_Additive
    translationMode=BMM_Additive
    alphaBoolBlend {blendOption=Linear}
    alphaScaleBiasClamp {bInterpResult=true, interpSpeedDecreasing=2, interpSpeedIncreasing=18}
    boneToModify {boneName=Controller}
    translation {y=2}
   }
   Output Pose(Result=Component To Local(ComponentPose=@n34)) {layerGroup=None}
   AnimGraphNode_StateMachine_0 = FrontEnd() {is=title AnimGraphNode_StateMachine}
   AnimGraphNode_StateMachine_1 = InGame() {is=title AnimGraphNode_StateMachine}
   n35 = Slot 'BearFrontend'(Source=Blend Poses by bool(BlendPose_0=@AnimGraphNode_StateMachine_0, BlendPose_1=@AnimGraphNode_StateMachine_1, bActiveValue=Get bIsInFrontEnd()) {blendType=Linear}) {slotName=BearFrontend}
   Save cached pose 'CachedPose_PreControlRig'(Pose=Layered blend per bone(
     BasePose=Copy Pose From Mesh() {bCopyCurves=true, bUseAttachedParent=true}
     BlendPoses_0=Slot 'BearOverride'(Source=@n35) {slotName=BearOverride}
    ) {layerSetup [{branchFilters [{boneName=bb_main}]}]}) {cacheName=CachedPose_PreControlRig}
   GetActorForwardVector!(self:Actor Object Reference=Get Player Pawn Athena() -> (Player Pawn Athena:Player Pawn Athena Object Reference)) -> (ReturnValue:Vector)
  ]}
 FrontEnd {
  entry=Idle
  path=AnimGraph.AnimGraphNode_StateMachine_0.FrontEnd
  ids=id
  nodes [
   AnimStateNode_0 = Idle()
   AnimStateNode_1 = Skydive() {is=title AnimStateNode}
   AnimStateTransitionNode_0 = AnimStateTransitionNode()
   AnimStateTransitionNode_1 = AnimStateTransitionNode()
  ]
  links [
   {
    dst=Skydive
    rule=AnimGraph.AnimGraphNode_StateMachine_0.FrontEnd.AnimStateTransitionNode_0.Transition
    src=Idle
    via=AnimStateTransitionNode_0
   }
   {
    dst=Idle
    rule=AnimGraph.AnimGraphNode_StateMachine_0.FrontEnd.AnimStateTransitionNode_1.Transition
    src=Skydive
    via=AnimStateTransitionNode_1
   }
  ]
 }
 Idle {
  path=AnimGraph.AnimGraphNode_StateMachine_0.FrontEnd.AnimStateNode_0.Idle
  nodes [Output Animation Pose(Result=Sequence Player 'Idle'() {sequence=/BRCosmetics/Characters/Player/Male/Large/Bodies/M_LRG_BadBear/Meshes/Parts/proto_anim/Idle})]
 }
 Skydive {
  path=AnimGraph.AnimGraphNode_StateMachine_0.FrontEnd.AnimStateNode_1.Skydive
  nodes [Output Animation Pose(Result=Sequence Player 'Skydive_Dive'() {sequence=/BRCosmetics/Characters/Player/Male/Large/Bodies/M_LRG_BadBear/Meshes/Parts/proto_anim/Skydive_Dive})]
 }
 Transition {
  path=AnimGraph.AnimGraphNode_StateMachine_0.FrontEnd.AnimStateTransitionNode_0.Transition
  nodes [Result(bCanEnterTransition=Get bIsSkyDiving())]
 }
 Transition {
  path=AnimGraph.AnimGraphNode_StateMachine_0.FrontEnd.AnimStateTransitionNode_1.Transition
  nodes [Result(bCanEnterTransition=NOT Boolean(A=Get bIsSkyDiving()))]
 }
 InGame {
  entry=Main
  path=AnimGraph.AnimGraphNode_StateMachine_1.InGame
  ids=id
  nodes [
   AnimStateNode_0 = Main() {is=title AnimStateNode}
   AnimStateNode_1 = DBNO() {is=title AnimStateNode}
   AnimStateNode_2 = DBNO Exit() {is=title AnimStateNode}
   AnimStateNode_3 = Skydive()
   AnimStateTransitionNode_0 = AnimStateTransitionNode()
   AnimStateTransitionNode_1 = AnimStateTransitionNode()
   AnimStateTransitionNode_2 = AnimStateTransitionNode()
   AnimStateTransitionNode_3 = AnimStateTransitionNode()
   AnimStateTransitionNode_4 = AnimStateTransitionNode() {bAutomaticRuleBasedOnSequencePlayerInState=true}
  ]
  links [
   {
    dst=DBNO Exit
    rule=AnimGraph.AnimGraphNode_StateMachine_1.InGame.AnimStateTransitionNode_3.Transition
    src=DBNO
    via=AnimStateTransitionNode_3
   }
   {
    dst=Main
    rule=AnimGraph.AnimGraphNode_StateMachine_1.InGame.AnimStateTransitionNode_4.Transition
    src=DBNO Exit
    via=AnimStateTransitionNode_4
   }
   {dst=DBNO, rule=AnimGraph.AnimGraphNode_StateMachine_1.InGame.AnimStateTransitionNode_0.Transition, src=Main, via=AnimStateTransitionNode_0}
   {
    dst=Skydive
    rule=AnimGraph.AnimGraphNode_StateMachine_1.InGame.AnimStateTransitionNode_1.Transition
    src=Main
    via=AnimStateTransitionNode_1
   }
   {
    dst=Main
    rule=AnimGraph.AnimGraphNode_StateMachine_1.InGame.AnimStateTransitionNode_2.Transition
    src=Skydive
    via=AnimStateTransitionNode_2
   }
  ]
 }
 Main {
  path=AnimGraph.AnimGraphNode_StateMachine_1.InGame.AnimStateNode_0.Main
  nodes [
   AnimGraphNode_StateMachine_0 = Main() {is=title AnimGraphNode_StateMachine}
   Output Animation Pose(Result=@AnimGraphNode_StateMachine_0)
  ]
 }
 Main {
  entry=Locomotion
  path=AnimGraph.AnimGraphNode_StateMachine_1.InGame.AnimStateNode_0.Main.AnimGraphNode_StateMachine_0.Main
  ids=id
  nodes [
   AnimStateConduitNode_0 = IsNotOnGround() {is=title AnimStateConduitNode}
   AnimStateNode_0 = Locomotion() {is=title AnimStateNode}
   AnimStateNode_1 = Jump_Start() {is=title AnimStateNode}
   AnimStateNode_2 = Jump_Fall() {is=title AnimStateNode}
   AnimStateNode_3 = Jump_Land() {is=title AnimStateNode}
   AnimStateNode_4 = Swimming() {is=title AnimStateNode}
   AnimStateNode_5 = Jump() {is=title AnimStateNode}
   AnimStateTransitionNode_0 = AnimStateTransitionNode()
   AnimStateTransitionNode_11 = AnimStateTransitionNode()
   AnimStateTransitionNode_12 = AnimStateTransitionNode()
   AnimStateTransitionNode_13 = AnimStateTransitionNode()
   AnimStateTransitionNode_3 = AnimStateTransitionNode()
   AnimStateTransitionNode_5 = AnimStateTransitionNode() {bAutomaticRuleBasedOnSequencePlayerInState=true}
   AnimStateTransitionNode_6 = AnimStateTransitionNode()
   AnimStateTransitionNode_7 = AnimStateTransitionNode()
   AnimStateTransitionNode_8 = AnimStateTransitionNode()
  ]
  links [
   {
    dst=Jump
    rule=AnimGraph.AnimGraphNode_StateMachine_1.InGame.AnimStateNode_0.Main.AnimGraphNode_StateMachine_0.Main.AnimStateTransitionNode_8.Transition
    src=IsNotOnGround
    via=AnimStateTransitionNode_8
   }
   {
    dst=Swimming
    rule=AnimGraph.AnimGraphNode_StateMachine_1.InGame.AnimStateNode_0.Main.AnimGraphNode_StateMachine_0.Main.AnimStateTransitionNode_6.Transition
    src=IsNotOnGround
    via=AnimStateTransitionNode_6
   }
   {
    dst=IsNotOnGround
    rule=AnimGraph.AnimGraphNode_StateMachine_1.InGame.AnimStateNode_0.Main.AnimGraphNode_StateMachine_0.Main.AnimStateTransitionNode_11.Transition
    src=Jump
    via=AnimStateTransitionNode_11
   }
   {
    dst=Jump_Land
    rule=AnimGraph.AnimGraphNode_StateMachine_1.InGame.AnimStateNode_0.Main.AnimGraphNode_StateMachine_0.Main.AnimStateTransitionNode_12.Transition
    src=Jump
    via=AnimStateTransitionNode_12
   }
   {
    dst=Locomotion
    rule=AnimGraph.AnimGraphNode_StateMachine_1.InGame.AnimStateNode_0.Main.AnimGraphNode_StateMachine_0.Main.AnimStateTransitionNode_5.Transition
    src=Jump_Land
    via=AnimStateTransitionNode_5
   }
   {
    dst=Jump_Fall
    rule=AnimGraph.AnimGraphNode_StateMachine_1.InGame.AnimStateNode_0.Main.AnimGraphNode_StateMachine_0.Main.AnimStateTransitionNode_3.AnimationTransitionGraph_0
    src=Jump_Start
    via=AnimStateTransitionNode_3
   }
   {
    dst=IsNotOnGround
    rule=AnimGraph.AnimGraphNode_StateMachine_1.InGame.AnimStateNode_0.Main.AnimGraphNode_StateMachine_0.Main.AnimStateTransitionNode_0.Transition
    src=Locomotion
    via=AnimStateTransitionNode_0
   }
   {
    dst=IsNotOnGround
    rule=AnimGraph.AnimGraphNode_StateMachine_1.InGame.AnimStateNode_0.Main.AnimGraphNode_StateMachine_0.Main.AnimStateTransitionNode_7.Transition
    src=Swimming
    via=AnimStateTransitionNode_7
   }
   {
    dst=Locomotion
    rule=AnimGraph.AnimGraphNode_StateMachine_1.InGame.AnimStateNode_0.Main.AnimGraphNode_StateMachine_0.Main.AnimStateTransitionNode_13.Transition
    src=Swimming
    via=AnimStateTransitionNode_13
   }
  ]
 }
 IsNotOnGround {
  path=AnimGraph.AnimGraphNode_StateMachine_1.InGame.AnimStateNode_0.Main.AnimGraphNode_StateMachine_0.Main.AnimStateConduitNode_0.IsNotOnGround
  nodes [Result(bCanEnterTransition=true)]
 }
 Locomotion {
  path=AnimGraph.AnimGraphNode_StateMachine_1.InGame.AnimStateNode_0.Main.AnimGraphNode_StateMachine_0.Main.AnimStateNode_0.Locomotion
  nodes [
   AnimGraphNode_StateMachine_0 = Locomotion() {is=title AnimGraphNode_StateMachine}
   Output Animation Pose(Result=@AnimGraphNode_StateMachine_0)
  ]
 }
 Locomotion {
  entry=Idle
  path=AnimGraph.AnimGraphNode_StateMachine_1.InGame.AnimStateNode_0.Main.AnimGraphNode_StateMachine_0.Main.AnimStateNode_0.Locomotion.AnimGraphNode_StateMachine_0.Locomotion
  ids=id
  nodes [
   AnimStateNode_0 = Idle()
   AnimStateNode_1 = Jog() {is=title AnimStateNode}
   AnimStateNode_2 = Slide() {is=title AnimStateNode}
   AnimStateTransitionNode_0 = AnimStateTransitionNode() {crossfadeDuration=0.4}
   AnimStateTransitionNode_1 = AnimStateTransitionNode() {crossfadeDuration=0.4}
   AnimStateTransitionNode_5 = AnimStateTransitionNode()
   AnimStateTransitionNode_6 = AnimStateTransitionNode()
  ]
  links [
   {
    dst=Jog
    rule=AnimGraph.AnimGraphNode_StateMachine_1.InGame.AnimStateNode_0.Main.AnimGraphNode_StateMachine_0.Main.AnimStateNode_0.Locomotion.AnimGraphNode_StateMachine_0.Locomotion.AnimStateTransitionNode_0.Transition
    src=Idle
    via=AnimStateTransitionNode_0
   }
   {
    dst=Idle
    rule=AnimGraph.AnimGraphNode_StateMachine_1.InGame.AnimStateNode_0.Main.AnimGraphNode_StateMachine_0.Main.AnimStateNode_0.Locomotion.AnimGraphNode_StateMachine_0.Locomotion.AnimStateTransitionNode_1.Transition
    src=Jog
    via=AnimStateTransitionNode_1
   }
   {
    dst=Slide
    rule=AnimGraph.AnimGraphNode_StateMachine_1.InGame.AnimStateNode_0.Main.AnimGraphNode_StateMachine_0.Main.AnimStateNode_0.Locomotion.AnimGraphNode_StateMachine_0.Locomotion.AnimStateTransitionNode_5.Transition
    src=Jog
    via=AnimStateTransitionNode_5
   }
   {
    dst=Jog
    rule=AnimGraph.AnimGraphNode_StateMachine_1.InGame.AnimStateNode_0.Main.AnimGraphNode_StateMachine_0.Main.AnimStateNode_0.Locomotion.AnimGraphNode_StateMachine_0.Locomotion.AnimStateTransitionNode_6.Transition
    src=Slide
    via=AnimStateTransitionNode_6
   }
  ]
 }
 Idle {
  path=AnimGraph.AnimGraphNode_StateMachine_1.InGame.AnimStateNode_0.Main.AnimGraphNode_StateMachine_0.Main.AnimStateNode_0.Locomotion.AnimGraphNode_StateMachine_0.Locomotion.AnimStateNode_0.Idle
  nodes [Output Animation Pose(Result=Blend Poses by bool(
     BlendPose_0=Sequence Player 'Crouch_idle'() {sequence=/BRCosmetics/Characters/Player/Male/Large/Bodies/M_LRG_BadBear/Meshes/Parts/proto_anim/Crouch_idle}
     BlendPose_1=Sequence Player 'Idle'() {sequence=&Idle}
     bActiveValue=Get bIsCrouching()
    ) {blendType=Linear})]
 }
 Jog {
  path=AnimGraph.AnimGraphNode_StateMachine_1.InGame.AnimStateNode_0.Main.AnimGraphNode_StateMachine_0.Main.AnimStateNode_0.Locomotion.AnimGraphNode_StateMachine_0.Locomotion.AnimStateNode_1.Jog
  nodes [
   Blend Poses by bool(
    BlendPose_0=Sequence Player 'Run_Forward'() {sequence=/BRCosmetics/Characters/Player/Male/Large/Bodies/M_LRG_BadBear/Meshes/Parts/proto_anim/Run_Forward}
    bActiveValue=Get bIsSprinting()
   ) {blendType=Linear}
   Sequence Player 'Temp_Locomotion_pose'() {sequence=/BRCosmetics/Characters/Player/Male/Large/Bodies/M_LRG_BadBear/Meshes/Parts/proto_anim/Temp_Locomotion_pose}
   n1 = Sequence Player 'Crouch_Run'() {sequence=/BRCosmetics/Characters/Player/Male/Large/Bodies/M_LRG_BadBear/Meshes/Parts/proto_anim/Crouch_Run}
   n2 = Sequence Player 'Idle_pose'() {sequence=/BRCosmetics/Characters/Player/Male/Large/Bodies/M_LRG_BadBear/Meshes/Parts/proto_anim/Idle_pose}
   n3 = Get bIsCrouching()
   n4 = "Blend Poses (EFortCardinalDirection)"!(
    BlendPose_0:Pose Link Structure=Sequence Player 'Idle_pose_ADD'() {sequence=/BRCosmetics/Characters/Player/Male/Large/Bodies/M_LRG_BadBear/Meshes/Parts/proto_anim/Idle_pose_ADD}
    BlendPose_1:Pose Link Structure=Sequence Player 'Run_Forward'() {sequence=&Run_Forward}
    BlendPose_2:Pose Link Structure=Sequence Player 'Run_R'() {sequence=/BRCosmetics/Characters/Player/Male/Large/Bodies/M_LRG_BadBear/Meshes/Parts/proto_anim/Run_R}
    BlendPose_3:Pose Link Structure=Sequence Player 'Run_Backward'() {sequence=/BRCosmetics/Characters/Player/Male/Large/Bodies/M_LRG_BadBear/Meshes/Parts/proto_anim/Run_Backward}
    BlendPose_4:Pose Link Structure=Sequence Player 'Run_L'() {sequence=/BRCosmetics/Characters/Player/Male/Large/Bodies/M_LRG_BadBear/Meshes/Parts/proto_anim/Run_L}
    BlendTime_0:float=0.3
    BlendTime_1:float=0.3
    BlendTime_2:float=0.3
    BlendTime_3:float=0.3
    BlendTime_4:float=0.3
    ActiveEnumValue:Byte=Get LocomotionCardinalDirection!(self:Fort Player Anim Instance Object Reference=Get Player AnimBP() -> (Player AnimBP:Fortnite M Avg Player Anim Blueprint Object Reference)) -> (LocomotionCardinalDirection:EFortCardinalDirection Enum=North)
   ) -> (Pose:Pose Link Structure) {
    is=title AnimGraphNode_BlendListByEnum
    blendType=Linear
    blendPose [
     {linkId=-1, sourceLinkId=-1}
     {linkId=-1, sourceLinkId=-1}
     {linkId=-1, sourceLinkId=-1}
     {linkId=-1, sourceLinkId=-1}
     {linkId=-1, sourceLinkId=-1}
    ]
    blendTime [0.1,0.1,0.1,0.1,0.1]
   }
   Output Animation Pose(Result=Apply Additive(Base=Blend Poses by bool(BlendPose_0=@n1, BlendPose_1=@n2, bActiveValue=@n3) {blendType=Linear}, Additive=@n4) {alphaBoolBlend {blendOption=Linear}})
  ]
 }
 Slide {
  path=AnimGraph.AnimGraphNode_StateMachine_1.InGame.AnimStateNode_0.Main.AnimGraphNode_StateMachine_0.Main.AnimStateNode_0.Locomotion.AnimGraphNode_StateMachine_0.Locomotion.AnimStateNode_2.Slide
  nodes [
   AnimGraphNode_StateMachine_0 = Slide() {is=title AnimGraphNode_StateMachine}
   Output Animation Pose(Result=@AnimGraphNode_StateMachine_0)
  ]
 }
 Slide {
  entry=Slide Start
  path=AnimGraph.AnimGraphNode_StateMachine_1.InGame.AnimStateNode_0.Main.AnimGraphNode_StateMachine_0.Main.AnimStateNode_0.Locomotion.AnimGraphNode_StateMachine_0.Locomotion.AnimStateNode_2.Slide.AnimGraphNode_StateMachine_0.Slide
  ids=id
  nodes [
   AnimStateNode_0 = Slide Start() {is=title AnimStateNode}
   AnimStateNode_1 = Slide Loop() {is=title AnimStateNode}
   AnimStateTransitionNode_7 = AnimStateTransitionNode() {bAutomaticRuleBasedOnSequencePlayerInState=true}
  ]
  links [{
    dst=Slide Loop
    rule=AnimGraph.AnimGraphNode_StateMachine_1.InGame.AnimStateNode_0.Main.AnimGraphNode_StateMachine_0.Main.AnimStateNode_0.Locomotion.AnimGraphNode_StateMachine_0.Locomotion.AnimStateNode_2.Slide.AnimGraphNode_StateMachine_0.Slide.AnimStateTransitionNode_7.Transition
    src=Slide Start
    via=AnimStateTransitionNode_7
   }]
 }
 Slide Start {
  path=AnimGraph.AnimGraphNode_StateMachine_1.InGame.AnimStateNode_0.Main.AnimGraphNode_StateMachine_0.Main.AnimStateNode_0.Locomotion.AnimGraphNode_StateMachine_0.Locomotion.AnimStateNode_2.Slide.AnimGraphNode_StateMachine_0.Slide.AnimStateNode_0.Slide Start
  nodes [
   n1 = Sequence Player 'KneeSlide_start'!(bLoopAnimation:bool=False) -> (Pose) {bLoopAnimation=false, sequence=/BRCosmetics/Characters/Player/Male/Large/Bodies/M_LRG_BadBear/Meshes/Parts/proto_anim/KneeSlide_start}
   Output Animation Pose(Result=@n1)
  ]
 }
 Slide Loop {
  path=AnimGraph.AnimGraphNode_StateMachine_1.InGame.AnimStateNode_0.Main.AnimGraphNode_StateMachine_0.Main.AnimStateNode_0.Locomotion.AnimGraphNode_StateMachine_0.Locomotion.AnimStateNode_2.Slide.AnimGraphNode_StateMachine_0.Slide.AnimStateNode_1.Slide Loop
  nodes [Output Animation Pose(Result=Sequence Player 'KneeSlidingPose'() {sequence=/BRCosmetics/Characters/Player/Male/Large/Bodies/M_LRG_BadBear/Meshes/Parts/proto_anim/KneeSlidingPose})]
 }
 Transition {
  path=AnimGraph.AnimGraphNode_StateMachine_1.InGame.AnimStateNode_0.Main.AnimGraphNode_StateMachine_0.Main.AnimStateNode_0.Locomotion.AnimGraphNode_StateMachine_0.Locomotion.AnimStateNode_2.Slide.AnimGraphNode_StateMachine_0.Slide.AnimStateTransitionNode_7.Transition
  nodes [Result()]
 }
 Transition {
  path=AnimGraph.AnimGraphNode_StateMachine_1.InGame.AnimStateNode_0.Main.AnimGraphNode_StateMachine_0.Main.AnimStateNode_0.Locomotion.AnimGraphNode_StateMachine_0.Locomotion.AnimStateTransitionNode_0.Transition
  nodes [Result(bCanEnterTransition=Get bIsMoving2D())]
 }
 Transition {
  path=AnimGraph.AnimGraphNode_StateMachine_1.InGame.AnimStateNode_0.Main.AnimGraphNode_StateMachine_0.Main.AnimStateNode_0.Locomotion.AnimGraphNode_StateMachine_0.Locomotion.AnimStateTransitionNode_1.Transition
  nodes [Result(bCanEnterTransition=NOT Boolean(A=Get bIsMoving2D()))]
 }
 Transition {
  path=AnimGraph.AnimGraphNode_StateMachine_1.InGame.AnimStateNode_0.Main.AnimGraphNode_StateMachine_0.Main.AnimStateNode_0.Locomotion.AnimGraphNode_StateMachine_0.Locomotion.AnimStateTransitionNode_5.Transition
  nodes [Result(bCanEnterTransition=Get bIsSliding!(self:Fort Anim Instance Object Reference=Get Player AnimBP()) -> (bIsSliding))]
 }
 Transition {
  path=AnimGraph.AnimGraphNode_StateMachine_1.InGame.AnimStateNode_0.Main.AnimGraphNode_StateMachine_0.Main.AnimStateNode_0.Locomotion.AnimGraphNode_StateMachine_0.Locomotion.AnimStateTransitionNode_6.Transition
  nodes [
   n1 = Get Player AnimBP()
   Result(bCanEnterTransition=NOT Boolean(A=Get bIsSliding(self=@n1)))
  ]
 }
 Jump_Start {
  path=AnimGraph.AnimGraphNode_StateMachine_1.InGame.AnimStateNode_0.Main.AnimGraphNode_StateMachine_0.Main.AnimStateNode_1.Jump_Start
  nodes [Output Animation Pose(Result=Sequence Player 'Jump_Start'() {bLoopAnimation=false, sequence=/BRCosmetics/Characters/Player/Male/Large/Bodies/M_LRG_BadBear/Meshes/Parts/proto_anim/Jump_Start})]
 }
 Jump_Fall {
  path=AnimGraph.AnimGraphNode_StateMachine_1.InGame.AnimStateNode_0.Main.AnimGraphNode_StateMachine_0.Main.AnimStateNode_2.Jump_Fall
  nodes [Output Animation Pose(Result=Sequence Player 'Jump_Fall'() {bLoopAnimation=false, sequence=/BRCosmetics/Characters/Player/Male/Large/Bodies/M_LRG_BadBear/Meshes/Parts/proto_anim/Jump_Fall})]
 }
 Jump_Land {
  path=AnimGraph.AnimGraphNode_StateMachine_1.InGame.AnimStateNode_0.Main.AnimGraphNode_StateMachine_0.Main.AnimStateNode_3.Jump_Land
  nodes [Output Animation Pose(Result=Sequence Player 'Jump_Land'() {bLoopAnimation=false, sequence=/BRCosmetics/Characters/Player/Male/Large/Bodies/M_LRG_BadBear/Meshes/Parts/proto_anim/Jump_Land})]
 }
 Swimming {
  path=AnimGraph.AnimGraphNode_StateMachine_1.InGame.AnimStateNode_0.Main.AnimGraphNode_StateMachine_0.Main.AnimStateNode_4.Swimming
  nodes [
   AnimGraphNode_StateMachine_0 = Swimming() {is=title AnimGraphNode_StateMachine}
   Output Animation Pose(Result=@AnimGraphNode_StateMachine_0)
  ]
 }
 Swimming {
  entry=Swimming
  path=AnimGraph.AnimGraphNode_StateMachine_1.InGame.AnimStateNode_0.Main.AnimGraphNode_StateMachine_0.Main.AnimStateNode_4.Swimming.AnimGraphNode_StateMachine_0.Swimming
  ids=id
  nodes [
   AnimStateNode_0 = Swimming() {is=title AnimStateNode}
   AnimStateNode_1 = Swimming Sprint() {is=title AnimStateNode}
   AnimStateTransitionNode_3 = AnimStateTransitionNode()
   AnimStateTransitionNode_4 = AnimStateTransitionNode()
  ]
  links [
   {
    dst=Swimming Sprint
    rule=AnimGraph.AnimGraphNode_StateMachine_1.InGame.AnimStateNode_0.Main.AnimGraphNode_StateMachine_0.Main.AnimStateNode_4.Swimming.AnimGraphNode_StateMachine_0.Swimming.AnimStateTransitionNode_3.Transition
    src=Swimming
    via=AnimStateTransitionNode_3
   }
   {
    dst=Swimming
    rule=AnimGraph.AnimGraphNode_StateMachine_1.InGame.AnimStateNode_0.Main.AnimGraphNode_StateMachine_0.Main.AnimStateNode_4.Swimming.AnimGraphNode_StateMachine_0.Swimming.AnimStateTransitionNode_4.Transition
    src=Swimming Sprint
    via=AnimStateTransitionNode_4
   }
  ]
 }
 Swimming {
  path=AnimGraph.AnimGraphNode_StateMachine_1.InGame.AnimStateNode_0.Main.AnimGraphNode_StateMachine_0.Main.AnimStateNode_4.Swimming.AnimGraphNode_StateMachine_0.Swimming.AnimStateNode_0.Swimming
  nodes [
   AnimGraphNode_StateMachine_0 = SwimmingLocomotion() {is=title AnimGraphNode_StateMachine}
   Output Animation Pose(Result=@AnimGraphNode_StateMachine_0)
  ]
 }
 SwimmingLocomotion {
  entry=SwimmingIdle
  path=AnimGraph.AnimGraphNode_StateMachine_1.InGame.AnimStateNode_0.Main.AnimGraphNode_StateMachine_0.Main.AnimStateNode_4.Swimming.AnimGraphNode_StateMachine_0.Swimming.AnimStateNode_0.Swimming.AnimGraphNode_StateMachine_0.SwimmingLocomotion
  ids=id
  nodes [
   AnimStateNode_0 = SwimmingIdle() {is=title AnimStateNode}
   AnimStateNode_1 = SwimmingMoving() {is=title AnimStateNode}
   AnimStateTransitionNode_0 = AnimStateTransitionNode()
   AnimStateTransitionNode_1 = AnimStateTransitionNode()
  ]
  links [
   {
    dst=SwimmingMoving
    rule=AnimGraph.AnimGraphNode_StateMachine_1.InGame.AnimStateNode_0.Main.AnimGraphNode_StateMachine_0.Main.AnimStateNode_4.Swimming.AnimGraphNode_StateMachine_0.Swimming.AnimStateNode_0.Swimming.AnimGraphNode_StateMachine_0.SwimmingLocomotion.AnimStateTransitionNode_0.Transition
    src=SwimmingIdle
    via=AnimStateTransitionNode_0
   }
   {
    dst=SwimmingIdle
    rule=AnimGraph.AnimGraphNode_StateMachine_1.InGame.AnimStateNode_0.Main.AnimGraphNode_StateMachine_0.Main.AnimStateNode_4.Swimming.AnimGraphNode_StateMachine_0.Swimming.AnimStateNode_0.Swimming.AnimGraphNode_StateMachine_0.SwimmingLocomotion.AnimStateTransitionNode_1.Transition
    src=SwimmingMoving
    via=AnimStateTransitionNode_1
   }
  ]
 }
 SwimmingIdle {
  path=AnimGraph.AnimGraphNode_StateMachine_1.InGame.AnimStateNode_0.Main.AnimGraphNode_StateMachine_0.Main.AnimStateNode_4.Swimming.AnimGraphNode_StateMachine_0.Swimming.AnimStateNode_0.Swimming.AnimGraphNode_StateMachine_0.SwimmingLocomotion.AnimStateNode_0.SwimmingIdle
  nodes [Output Animation Pose(Result=Sequence Player 'swim_idle'() {sequence=/BRCosmetics/Characters/Player/Male/Large/Bodies/M_LRG_BadBear/Meshes/Parts/proto_anim/swim_idle})]
 }
 SwimmingMoving {
  path=AnimGraph.AnimGraphNode_StateMachine_1.InGame.AnimStateNode_0.Main.AnimGraphNode_StateMachine_0.Main.AnimStateNode_4.Swimming.AnimGraphNode_StateMachine_0.Swimming.AnimStateNode_0.Swimming.AnimGraphNode_StateMachine_0.SwimmingLocomotion.AnimStateNode_1.SwimmingMoving
  nodes [
   n1 = "Blend Poses (EFortCardinalDirection)"(
    BlendPose_0=Sequence Player 'Idle_pose_ADD'() {sequence=&Idle_pose_ADD}
    BlendPose_1=Sequence Player 'Run_Forward'() {sequence=&Run_Forward}
    BlendPose_2=Sequence Player 'Run_R'() {sequence=&Run_R}
    BlendPose_3=Sequence Player 'Run_Backward'() {sequence=&Run_Backward}
    BlendPose_4=Sequence Player 'Run_L'() {sequence=&Run_L}
    BlendTime_0=0.3
    BlendTime_1=0.3
    BlendTime_2=0.3
    BlendTime_3=0.3
    BlendTime_4=0.3
    ActiveEnumValue=Get LocomotionCardinalDirection(self=Get Player AnimBP()) -> (LocomotionCardinalDirection=North)
   ) {
    blendType=Linear
    blendPose [
     {linkId=-1, sourceLinkId=-1}
     {linkId=-1, sourceLinkId=-1}
     {linkId=-1, sourceLinkId=-1}
     {linkId=-1, sourceLinkId=-1}
     {linkId=-1, sourceLinkId=-1}
    ]
    blendTime [0.1,0.1,0.1,0.1,0.1]
   }
   Output Animation Pose(Result=Apply Additive(
     Base=Sequence Player 'swim_idle'() {sequence=&swim_idle}
     Additive=@n1
    ) {alphaBoolBlend {blendOption=Linear}})
  ]
 }
 Transition {
  path=AnimGraph.AnimGraphNode_StateMachine_1.InGame.AnimStateNode_0.Main.AnimGraphNode_StateMachine_0.Main.AnimStateNode_4.Swimming.AnimGraphNode_StateMachine_0.Swimming.AnimStateNode_0.Swimming.AnimGraphNode_StateMachine_0.SwimmingLocomotion.AnimStateTransitionNode_0.Transition
  nodes [Result(bCanEnterTransition=Get bIsMoving2D())]
 }
 Transition {
  path=AnimGraph.AnimGraphNode_StateMachine_1.InGame.AnimStateNode_0.Main.AnimGraphNode_StateMachine_0.Main.AnimStateNode_4.Swimming.AnimGraphNode_StateMachine_0.Swimming.AnimStateNode_0.Swimming.AnimGraphNode_StateMachine_0.SwimmingLocomotion.AnimStateTransitionNode_1.Transition
  nodes [Result(bCanEnterTransition=NOT Boolean(A=Get bIsMoving2D()))]
 }
 Swimming Sprint {
  path=AnimGraph.AnimGraphNode_StateMachine_1.InGame.AnimStateNode_0.Main.AnimGraphNode_StateMachine_0.Main.AnimStateNode_4.Swimming.AnimGraphNode_StateMachine_0.Swimming.AnimStateNode_1.Swimming Sprint
  nodes [Output Animation Pose(Result=Sequence Player 'Swim_N'() {sequence=/BRCosmetics/Characters/Player/Male/Large/Bodies/M_LRG_BadBear/Meshes/Parts/proto_anim/Swim_N})]
 }
 Transition {
  path=AnimGraph.AnimGraphNode_StateMachine_1.InGame.AnimStateNode_0.Main.AnimGraphNode_StateMachine_0.Main.AnimStateNode_4.Swimming.AnimGraphNode_StateMachine_0.Swimming.AnimStateTransitionNode_3.Transition
  nodes [
   Result(bCanEnterTransition=Get bIsSwimSprinting() -> (bIsSwimSprinting:bool))
   Get bIsAccelerating2D()
   Get bIsMoving2D()
  ]
 }
 Transition {
  path=AnimGraph.AnimGraphNode_StateMachine_1.InGame.AnimStateNode_0.Main.AnimGraphNode_StateMachine_0.Main.AnimStateNode_4.Swimming.AnimGraphNode_StateMachine_0.Swimming.AnimStateTransitionNode_4.Transition
  nodes [
   Result(bCanEnterTransition=NOT Boolean(A=Get bIsAccelerating2D()))
   Get bIsMoving2D()
  ]
 }
 Jump {
  path=AnimGraph.AnimGraphNode_StateMachine_1.InGame.AnimStateNode_0.Main.AnimGraphNode_StateMachine_0.Main.AnimStateNode_5.Jump
  nodes [
   AnimGraphNode_StateMachine_0 = Jump() {is=title AnimGraphNode_StateMachine}
   Output Animation Pose(Result=@AnimGraphNode_StateMachine_0)
  ]
 }
 Jump {
  entry=Jump_Start
  path=AnimGraph.AnimGraphNode_StateMachine_1.InGame.AnimStateNode_0.Main.AnimGraphNode_StateMachine_0.Main.AnimStateNode_5.Jump.AnimGraphNode_StateMachine_0.Jump
  ids=id
  nodes [
   AnimStateNode_1 = Jump_Start()
   AnimStateNode_2 = Jump_Fall()
   AnimStateTransitionNode_3 = AnimStateTransitionNode()
  ]
  links [{
    dst=Jump_Fall
    rule=AnimGraph.AnimGraphNode_StateMachine_1.InGame.AnimStateNode_0.Main.AnimGraphNode_StateMachine_0.Main.AnimStateNode_5.Jump.AnimGraphNode_StateMachine_0.Jump.AnimStateTransitionNode_3.AnimationTransitionGraph_0
    src=Jump_Start
    via=AnimStateTransitionNode_3
   }]
 }
 Jump_Start {
  path=AnimGraph.AnimGraphNode_StateMachine_1.InGame.AnimStateNode_0.Main.AnimGraphNode_StateMachine_0.Main.AnimStateNode_5.Jump.AnimGraphNode_StateMachine_0.Jump.AnimStateNode_1.Jump_Start
  nodes [Output Animation Pose(Result=Sequence Player 'Jump_Start'() {bLoopAnimation=false, sequence=&Jump_Start})]
 }
 Jump_Fall {
  path=AnimGraph.AnimGraphNode_StateMachine_1.InGame.AnimStateNode_0.Main.AnimGraphNode_StateMachine_0.Main.AnimStateNode_5.Jump.AnimGraphNode_StateMachine_0.Jump.AnimStateNode_2.Jump_Fall
  nodes [Output Animation Pose(Result=Sequence Player 'Jump_Fall'() {bLoopAnimation=false, sequence=&Jump_Fall})]
 }
 AnimationTransitionGraph_0 {
  path=AnimGraph.AnimGraphNode_StateMachine_1.InGame.AnimStateNode_0.Main.AnimGraphNode_StateMachine_0.Main.AnimStateNode_5.Jump.AnimGraphNode_StateMachine_0.Jump.AnimStateTransitionNode_3.AnimationTransitionGraph_0
  nodes [
   n1 = Get bIsOnGround() -> (bIsOnGround:bool)
   n2 = Get PawnVelocityZ()
   Result(bCanEnterTransition=AND Boolean(A=NOT Boolean(A=@n1), B="float < float"(A=@n2)))
   Get bIsFalling()
  ]
 }
 Transition {
  path=AnimGraph.AnimGraphNode_StateMachine_1.InGame.AnimStateNode_0.Main.AnimGraphNode_StateMachine_0.Main.AnimStateTransitionNode_0.Transition
  nodes [Result(bCanEnterTransition=NOT Boolean(A=Get bIsOnGround()))]
 }
 Transition {
  path=AnimGraph.AnimGraphNode_StateMachine_1.InGame.AnimStateNode_0.Main.AnimGraphNode_StateMachine_0.Main.AnimStateTransitionNode_11.Transition
  nodes [Result(bCanEnterTransition=Get bIsSurfaceSwimming())]
 }
 Transition {
  path=AnimGraph.AnimGraphNode_StateMachine_1.InGame.AnimStateNode_0.Main.AnimGraphNode_StateMachine_0.Main.AnimStateTransitionNode_12.Transition
  nodes [Result(bCanEnterTransition=Get bIsOnGround())]
 }
 Transition {
  path=AnimGraph.AnimGraphNode_StateMachine_1.InGame.AnimStateNode_0.Main.AnimGraphNode_StateMachine_0.Main.AnimStateTransitionNode_13.Transition
  nodes [Result(bCanEnterTransition=Get bIsOnGround())]
 }
 AnimationTransitionGraph_0 {
  path=AnimGraph.AnimGraphNode_StateMachine_1.InGame.AnimStateNode_0.Main.AnimGraphNode_StateMachine_0.Main.AnimStateTransitionNode_3.AnimationTransitionGraph_0
  nodes [
   n1 = Get bIsOnGround()
   n2 = Get PawnVelocityZ()
   Result(bCanEnterTransition=AND Boolean(A=NOT Boolean(A=@n1), B="float < float"(A=@n2)))
   Get bIsFalling()
  ]
 }
 Transition {
  path=AnimGraph.AnimGraphNode_StateMachine_1.InGame.AnimStateNode_0.Main.AnimGraphNode_StateMachine_0.Main.AnimStateTransitionNode_5.Transition
  nodes [Result()]
 }
 Transition {
  path=AnimGraph.AnimGraphNode_StateMachine_1.InGame.AnimStateNode_0.Main.AnimGraphNode_StateMachine_0.Main.AnimStateTransitionNode_6.Transition
  nodes [Result(bCanEnterTransition=Get bIsSurfaceSwimming())]
 }
 Transition {
  path=AnimGraph.AnimGraphNode_StateMachine_1.InGame.AnimStateNode_0.Main.AnimGraphNode_StateMachine_0.Main.AnimStateTransitionNode_7.Transition
  nodes [Result(bCanEnterTransition=NOT Boolean(A=Get bIsSurfaceSwimming()))]
 }
 Transition {
  path=AnimGraph.AnimGraphNode_StateMachine_1.InGame.AnimStateNode_0.Main.AnimGraphNode_StateMachine_0.Main.AnimStateTransitionNode_8.Transition
  nodes [Result(bCanEnterTransition=NOT Boolean(A=Get bIsSurfaceSwimming()))]
 }
 DBNO {
  path=AnimGraph.AnimGraphNode_StateMachine_1.InGame.AnimStateNode_1.DBNO
  nodes [
   AnimGraphNode_StateMachine_0 = DBNO() {is=title AnimGraphNode_StateMachine}
   Output Animation Pose(Result=@AnimGraphNode_StateMachine_0)
  ]
 }
 DBNO {
  entry=DBNO Start
  path=AnimGraph.AnimGraphNode_StateMachine_1.InGame.AnimStateNode_1.DBNO.AnimGraphNode_StateMachine_0.DBNO
  ids=id
  nodes [
   AnimStateNode_0 = DBNO Start() {is=title AnimStateNode}
   AnimStateNode_1 = DBNO Loop() {is=title AnimStateNode}
   AnimStateTransitionNode_2 = AnimStateTransitionNode() {bAutomaticRuleBasedOnSequencePlayerInState=true}
  ]
  links [{
    dst=DBNO Loop
    rule=AnimGraph.AnimGraphNode_StateMachine_1.InGame.AnimStateNode_1.DBNO.AnimGraphNode_StateMachine_0.DBNO.AnimStateTransitionNode_2.Transition
    src=DBNO Start
    via=AnimStateTransitionNode_2
   }]
 }
 DBNO Start {
  path=AnimGraph.AnimGraphNode_StateMachine_1.InGame.AnimStateNode_1.DBNO.AnimGraphNode_StateMachine_0.DBNO.AnimStateNode_0.DBNO Start
  nodes [Output Animation Pose(Result=Sequence Player 'DBNO_Intro'!(bLoopAnimation:bool=false) -> (Pose) {sequence=/BRCosmetics/Characters/Player/Male/Large/Bodies/M_LRG_BadBear/Meshes/Parts/proto_anim/DBNO_Intro})]
 }
 DBNO Loop {
  path=AnimGraph.AnimGraphNode_StateMachine_1.InGame.AnimStateNode_1.DBNO.AnimGraphNode_StateMachine_0.DBNO.AnimStateNode_1.DBNO Loop
  nodes [Output Animation Pose(Result=Sequence Player 'DBNO_Idle'() {sequence=/BRCosmetics/Characters/Player/Male/Large/Bodies/M_LRG_BadBear/Meshes/Parts/proto_anim/DBNO_Idle})]
 }
 Transition {
  path=AnimGraph.AnimGraphNode_StateMachine_1.InGame.AnimStateNode_1.DBNO.AnimGraphNode_StateMachine_0.DBNO.AnimStateTransitionNode_2.Transition
  nodes [Result()]
 }
 DBNO Exit {
  path=AnimGraph.AnimGraphNode_StateMachine_1.InGame.AnimStateNode_2.DBNO Exit
  nodes [Output Animation Pose(Result=Sequence Player 'DBNO_Outro'() {bLoopAnimation=false, sequence=/BRCosmetics/Characters/Player/Male/Large/Bodies/M_LRG_BadBear/Meshes/Parts/proto_anim/DBNO_Outro})]
 }
 Skydive {
  path=AnimGraph.AnimGraphNode_StateMachine_1.InGame.AnimStateNode_3.Skydive
  nodes [
   Sequence Player 'Idle'() {sequence=&Idle}
   AnimGraphNode_StateMachine_0 = SkydiveSM() {is=title AnimGraphNode_StateMachine}
   n1 = Get Skydiving!(self:Fort Player Anim Instance Object Reference=Get Player AnimBP()) -> (Skydiving:Fort Anim Input Skydiving Structure)
   Output Animation Pose(Result=Blend Poses by bool(
     BlendPose_0=Sequence Player 'Skydive_parachute'() {sequence=/BRCosmetics/Characters/Player/Male/Large/Bodies/M_LRG_BadBear/Meshes/Parts/proto_anim/Skydive_parachute}
     BlendPose_1=Sequence Player 'Skydive_Dive'() {sequence=&Skydive_Dive}
     BlendTime_0=1.5
     BlendTime_1=0.25
     bActiveValue=Break Fort Anim Input Skydiving!(FortAnimInput_Skydiving:"Fort Anim Input Skydiving Structure (by ref)"=@n1) -> (
      ParaGilderRootOffset:Vector=~
      ParaGliderRootRotationOverride:Rotator=~
      bUseParaGlideRootModifier:bool=~
      bIsSkydivingFromLaunchPad:bool=~
      bIsSkydivingFromBus:bool=~
      bIsInVortex:bool=~
      bIsDiving:bool=~
      bIsParachuteOpen:bool=~
      bIsSkydiving:bool=~
      bIsParachuteLeaning:bool=~
      bIsSkydiveLeaning:bool=~
      bIsLeaning:bool=~
      bIsSkydiveDiveMode:bool=~
      bPlaySkydiveDrift:bool=~
      bSkydiveDriftDelayActive:bool=~
      bSkydiveDriftAnimAllowed:bool=~
      bIsGliderRight:bool=~
      bIsGliderCenter:bool=~
      bIsGliderLeft:bool=~
      bIsGliderForward:bool=~
      bIsGliderBack:bool=~
      bHasDeployedAndIsStandingUpright:bool=~
      LocalAccelForward:float=~
      LocalAccelRight:float=~
      LocalVelocityRight:float=~
      SkydiveAimPitch:float=~
      SkydiveAimYaw:float=~
      DeployChuteAnimRate:float=~
      SkydiveDriftAnimRate:float=~
      SkydiveDriftAnimRateCurrent:float=~
      SkydiveFidgetAnimRate:float=~
      SkydiveFidgetAnimRateCurrent:float=~
      SkydiveAdditiveAlpha:float=~
      SkydiveDriftDelay:float=~
      SkydiveDriftAnim:int
      SkydiveDriftAnimMax:int
      LaunchpadAnim:int
      LocalAccelDir:ESkydivingDirection Enum
      DirectionLast:ESkydivingDirection Enum
      PoseAdditive:Anim Sequence Object Reference
     ) {is=title K2Node_BreakStruct}.bIsParachuteOpen
    ) {blendType=Linear})
  ]
 }
 SkydiveSM {
  entry=Skydive
  path=AnimGraph.AnimGraphNode_StateMachine_1.InGame.AnimStateNode_3.Skydive.AnimGraphNode_StateMachine_0.SkydiveSM
  ids=id
  nodes [
   AnimStateNode_0 = Skydive()
   AnimStateNode_1 = OpenParachute() {is=title AnimStateNode}
   AnimStateNode_2 = Parachute() {is=title AnimStateNode}
   AnimStateTransitionNode_0 = AnimStateTransitionNode()
   AnimStateTransitionNode_1 = AnimStateTransitionNode() {bAutomaticRuleBasedOnSequencePlayerInState=true}
   AnimStateTransitionNode_2 = AnimStateTransitionNode()
  ]
  links [
   {
    dst=Parachute
    rule=AnimGraph.AnimGraphNode_StateMachine_1.InGame.AnimStateNode_3.Skydive.AnimGraphNode_StateMachine_0.SkydiveSM.AnimStateTransitionNode_1.Transition
    src=OpenParachute
    via=AnimStateTransitionNode_1
   }
   {
    dst=Skydive
    rule=AnimGraph.AnimGraphNode_StateMachine_1.InGame.AnimStateNode_3.Skydive.AnimGraphNode_StateMachine_0.SkydiveSM.AnimStateTransitionNode_2.Transition
    src=Parachute
    via=AnimStateTransitionNode_2
   }
   {
    dst=OpenParachute
    rule=AnimGraph.AnimGraphNode_StateMachine_1.InGame.AnimStateNode_3.Skydive.AnimGraphNode_StateMachine_0.SkydiveSM.AnimStateTransitionNode_0.Transition
    src=Skydive
    via=AnimStateTransitionNode_0
   }
  ]
 }
 Skydive {
  path=AnimGraph.AnimGraphNode_StateMachine_1.InGame.AnimStateNode_3.Skydive.AnimGraphNode_StateMachine_0.SkydiveSM.AnimStateNode_0.Skydive
  nodes [Output Animation Pose(Result=Sequence Player 'Skydive_Dive'() {sequence=&Skydive_Dive})]
 }
 OpenParachute {
  path=AnimGraph.AnimGraphNode_StateMachine_1.InGame.AnimStateNode_3.Skydive.AnimGraphNode_StateMachine_0.SkydiveSM.AnimStateNode_1.OpenParachute
  nodes [
   n1 = Sequence Player 'SkydiveToParachute_pose'!(bLoopAnimation:bool=False) -> (Pose) {
    bLoopAnimation=false
    sequence=/BRCosmetics/Characters/Player/Male/Large/Bodies/M_LRG_BadBear/Meshes/Parts/proto_anim/SkydiveToParachute_pose
   }
   Output Animation Pose(Result=@n1)
  ]
 }
 Parachute {
  path=AnimGraph.AnimGraphNode_StateMachine_1.InGame.AnimStateNode_3.Skydive.AnimGraphNode_StateMachine_0.SkydiveSM.AnimStateNode_2.Parachute
  nodes [Output Animation Pose(Result=Sequence Player 'Skydive_parachute'() {sequence=&Skydive_parachute})]
 }
 Transition {
  path=AnimGraph.AnimGraphNode_StateMachine_1.InGame.AnimStateNode_3.Skydive.AnimGraphNode_StateMachine_0.SkydiveSM.AnimStateTransitionNode_0.Transition
  nodes [
   n1 = Get Player AnimBP()
   Result(bCanEnterTransition=Break Fort Anim Input Skydiving(FortAnimInput_Skydiving=Get Skydiving(self=@n1)) -> (
     ParaGilderRootOffset=~
     ParaGliderRootRotationOverride=~
     bUseParaGlideRootModifier=~
     bIsSkydivingFromLaunchPad=~
     bIsSkydivingFromBus=~
     bIsInVortex=~
     bIsDiving=~
     bIsParachuteOpen=~
     bIsSkydiving=~
     bIsParachuteLeaning=~
     bIsSkydiveLeaning=~
     bIsLeaning=~
     bIsSkydiveDiveMode=~
     bPlaySkydiveDrift=~
     bSkydiveDriftDelayActive=~
     bSkydiveDriftAnimAllowed=~
     bIsGliderRight=~
     bIsGliderCenter=~
     bIsGliderLeft=~
     bIsGliderForward=~
     bIsGliderBack=~
     bHasDeployedAndIsStandingUpright=~
     LocalAccelForward=~
     LocalAccelRight=~
     LocalVelocityRight=~
     SkydiveAimPitch=~
     SkydiveAimYaw=~
     DeployChuteAnimRate=~
     SkydiveDriftAnimRate=~
     SkydiveDriftAnimRateCurrent=~
     SkydiveFidgetAnimRate=~
     SkydiveFidgetAnimRateCurrent=~
     SkydiveAdditiveAlpha=~
     SkydiveDriftDelay=~
    ).bIsParachuteOpen)
  ]
 }
 Transition {
  path=AnimGraph.AnimGraphNode_StateMachine_1.InGame.AnimStateNode_3.Skydive.AnimGraphNode_StateMachine_0.SkydiveSM.AnimStateTransitionNode_1.Transition
  nodes [Result()]
 }
 Transition {
  path=AnimGraph.AnimGraphNode_StateMachine_1.InGame.AnimStateNode_3.Skydive.AnimGraphNode_StateMachine_0.SkydiveSM.AnimStateTransitionNode_2.Transition
  nodes [
   n1 = Get Skydiving(self=Get Player AnimBP())
   Result(bCanEnterTransition=NOT Boolean(A=Break Fort Anim Input Skydiving(FortAnimInput_Skydiving=@n1) -> (
      ParaGilderRootOffset=~
      ParaGliderRootRotationOverride=~
      bUseParaGlideRootModifier=~
      bIsSkydivingFromLaunchPad=~
      bIsSkydivingFromBus=~
      bIsInVortex=~
      bIsDiving=~
      bIsParachuteOpen=~
      bIsSkydiving=~
      bIsParachuteLeaning=~
      bIsSkydiveLeaning=~
      bIsLeaning=~
      bIsSkydiveDiveMode=~
      bPlaySkydiveDrift=~
      bSkydiveDriftDelayActive=~
      bSkydiveDriftAnimAllowed=~
      bIsGliderRight=~
      bIsGliderCenter=~
      bIsGliderLeft=~
      bIsGliderForward=~
      bIsGliderBack=~
      bHasDeployedAndIsStandingUpright=~
      LocalAccelForward=~
      LocalAccelRight=~
      LocalVelocityRight=~
      SkydiveAimPitch=~
      SkydiveAimYaw=~
      DeployChuteAnimRate=~
      SkydiveDriftAnimRate=~
      SkydiveDriftAnimRateCurrent=~
      SkydiveFidgetAnimRate=~
      SkydiveFidgetAnimRateCurrent=~
      SkydiveAdditiveAlpha=~
      SkydiveDriftDelay=~
     ).bIsParachuteOpen))
  ]
 }
 Transition {path=AnimGraph.AnimGraphNode_StateMachine_1.InGame.AnimStateTransitionNode_0.Transition, nodes [Result(bCanEnterTransition=Get bIsDBNO())]}
 Transition {
  path=AnimGraph.AnimGraphNode_StateMachine_1.InGame.AnimStateTransitionNode_1.Transition
  nodes [
   n1 = Get Player AnimBP()
   Result(bCanEnterTransition=Break Fort Anim Input Skydiving(FortAnimInput_Skydiving=Get Skydiving(self=@n1)) -> (
     ParaGilderRootOffset=~
     ParaGliderRootRotationOverride=~
     bUseParaGlideRootModifier=~
     bIsSkydivingFromLaunchPad=~
     bIsSkydivingFromBus=~
     bIsInVortex=~
     bIsDiving=~
     bIsParachuteOpen=~
     bIsSkydiving=~
     bIsParachuteLeaning=~
     bIsSkydiveLeaning=~
     bIsLeaning=~
     bIsSkydiveDiveMode=~
     bPlaySkydiveDrift=~
     bSkydiveDriftDelayActive=~
     bSkydiveDriftAnimAllowed=~
     bIsGliderRight=~
     bIsGliderCenter=~
     bIsGliderLeft=~
     bIsGliderForward=~
     bIsGliderBack=~
     bHasDeployedAndIsStandingUpright=~
     LocalAccelForward=~
     LocalAccelRight=~
     LocalVelocityRight=~
     SkydiveAimPitch=~
     SkydiveAimYaw=~
     DeployChuteAnimRate=~
     SkydiveDriftAnimRate=~
     SkydiveDriftAnimRateCurrent=~
     SkydiveFidgetAnimRate=~
     SkydiveFidgetAnimRateCurrent=~
     SkydiveAdditiveAlpha=~
     SkydiveDriftDelay=~
    ).bIsSkydiving)
  ]
 }
 Transition {
  path=AnimGraph.AnimGraphNode_StateMachine_1.InGame.AnimStateTransitionNode_2.Transition
  nodes [
   n1 = Get Skydiving(self=Get Player AnimBP())
   Result(bCanEnterTransition=NOT Boolean(A=Break Fort Anim Input Skydiving(FortAnimInput_Skydiving=@n1) -> (
      ParaGilderRootOffset=~
      ParaGliderRootRotationOverride=~
      bUseParaGlideRootModifier=~
      bIsSkydivingFromLaunchPad=~
      bIsSkydivingFromBus=~
      bIsInVortex=~
      bIsDiving=~
      bIsParachuteOpen=~
      bIsSkydiving=~
      bIsParachuteLeaning=~
      bIsSkydiveLeaning=~
      bIsLeaning=~
      bIsSkydiveDiveMode=~
      bPlaySkydiveDrift=~
      bSkydiveDriftDelayActive=~
      bSkydiveDriftAnimAllowed=~
      bIsGliderRight=~
      bIsGliderCenter=~
      bIsGliderLeft=~
      bIsGliderForward=~
      bIsGliderBack=~
      bHasDeployedAndIsStandingUpright=~
      LocalAccelForward=~
      LocalAccelRight=~
      LocalVelocityRight=~
      SkydiveAimPitch=~
      SkydiveAimYaw=~
      DeployChuteAnimRate=~
      SkydiveDriftAnimRate=~
      SkydiveDriftAnimRateCurrent=~
      SkydiveFidgetAnimRate=~
      SkydiveFidgetAnimRateCurrent=~
      SkydiveAdditiveAlpha=~
      SkydiveDriftDelay=~
     ).bIsSkydiving))
  ]
 }
 Transition {
  path=AnimGraph.AnimGraphNode_StateMachine_1.InGame.AnimStateTransitionNode_3.Transition
  nodes [Result(bCanEnterTransition=NOT Boolean(A=Get bIsDBNO()))]
 }
 Transition {path=AnimGraph.AnimGraphNode_StateMachine_1.InGame.AnimStateTransitionNode_4.Transition, nodes [Result()]}
 EventGraph {nodes [
   n1 = Event BlueprintUpdateAnimation()
   > n2 = Sequence!(execute) -> (then_0, then_1, then_2:Exec, then_3:Exec, then_4:Exec, then_5:Exec, then_6:Exec, then_7:Exec, then_8:Exec, then_9:Exec)
   > n3 = Gravity Override State Function DEPRECATED(CharacterPartAnimInstance=Self-Reference(), GameTimeInSeconds=GetGameTimeInSeconds(), GravityOverrideParams=Get Gravity Override Params())
   > Set GravityOverride(GravityOverride=@n3)
   n4 = Branch(execute=@n2.then_1, Condition=IsNullItem!(ItemObject:Object Reference=Get Player Pawn Athena()) -> (Is Null Item:bool))
   > n5 = Cast To PlayerPawn_Athena(Object=GetOwningActor())
   > n6 = Set Player Pawn Athena(Player Pawn Athena:Player Pawn Athena Object Reference=@n5) -> (Output_Get:Player Pawn Athena Object Reference)
   n7 = Sequence!(execute=[@n4.else, @n6.then]) -> (then_0, then_1, then_2)
   n8 = Get bIsSkyDiving()
   > n9 = Set NotJumping(NotJumping:double="To Float (Boolean)"(InBool=NOT Boolean(A=@n8))) -> (Output_Get:double)
   n10 = "To Float (Boolean)"(InBool=NOT Boolean(A=Get bIsSurfaceSwimming()))
   n11 = "To Float (Boolean)"(InBool=NOT Boolean(A=Get bIsSwimSprinting()))
   > n12 = Set ControlRigAlpha(ControlRigAlpha:double=float * float(A=@n9, B=float * float!(A=@n11, B=@n10, C:double=1) -> (ReturnValue)) {is=title K2Node_PromotableOperator}) -> (Output_Get:double)
   > n13 = Set SpringStrength(SpringStrength:double=0.03) -> (Output_Get:double)
   > Set SpringAlpha(SpringAlpha:double=float * float!(A=@n12, B=@n13) -> (ReturnValue)) -> (Output_Get:double)
   n14 = Get bIsSurfaceSwimming()
   n15 = NOT Boolean(A=Get bIsSwimSprinting())
   Set SurfaceSwimAmount(
    execute=@n7.then_1
    SurfaceSwimAmount:double=float * float!(A="To Float (Boolean)"(InBool=@n14), B=0.5, C:double="To Float (Boolean)"(InBool=@n15)) -> (ReturnValue)
   ) -> (Output_Get:double)
   Reroute Node(InputPin=@n2.then_2)
   n16 = Get bIsSprinting()
   Set easingRunning(
    execute=@n2.then_3
    easingRunning:double=FInterpTo(Current=Get easingRunning() -> (easingRunning:double), Target="To Float (Boolean)"(InBool=@n16), DeltaTime=@n1.DeltaTimeX, InterpSpeed=2)
   ) -> (Output_Get:double)
   K2Node_Composite_0 = Anim!(execute=@n2.then_4) -> (OutputPin:Exec) {is=title K2Node_Composite}
   n17 = GetAnimInstance(self=GetOwnerSkeletalMeshComponent())
   Is Valid(exec=@n2.then_5, InputObject=@n17)
   > n18 = Set IsCustomEmotePlaying(IsCustomEmotePlaying:double=SelectFloat(A=1, B=0, bPickA=Montage_IsPlaying(self=@n17, Montage=/BRCosmetics/Animation/Game/MainPlayer/Emotes/BadBear/BadBearRobot_M))) -> (Output_Get:double)
   > Set CustomEmoteNotPlaying(CustomEmoteNotPlaying:double="To Float (Boolean)"(InBool="float < float"(A=@n18, B=0.5))) -> (Output_Get:double)
   > Set CustomEmoteBool(CustomEmoteBool:bool="float > float"(A=@n18, B=0.5)) -> (Output_Get:bool)
   n19 = NOT Boolean(A=Get bIsPlayingEmote())
   Set isInFrontendAndNotPlayingEmote(execute=@n2.then_6, isInFrontendAndNotPlayingEmote:bool=AND Boolean(A=Get bIsInFrontEnd(), B=@n19)) -> (Output_Get:bool)
   n20 = Branch(execute=@n2.then_7, Condition=AND Boolean(A=Get bIsInFrontEnd(), B=@n19))
   > Set FrontEndCustomEmote(FrontEndCustomEmote:double=1) -> (Output_Get:double)
   Set FrontEndCustomEmote(execute=@n20.else, FrontEndCustomEmote=Get IsCustomEmotePlaying() -> (IsCustomEmotePlaying:double))
   n21 = Get bIsInFrontEnd()
   n22 = AND Boolean(A=@n21, B=NOT Boolean(A=Get bIsPlayingEmote()))
   Set frontEndNoEmote(execute=@n2.then_8, frontEndNoEmote:bool=@n22) -> (Output_Get:bool)
   n23 = Get CustomEmoteBool() -> (CustomEmoteBool:bool)
   > Set FrontEndNoEmoteExceptCustomEmote(FrontEndNoEmoteExceptCustomEmote:bool=OR Boolean(A=AND Boolean(A=@n23, B=@n21), B=@n22)) -> (Output_Get:bool)
   n24 = Set pawnVelocitySpeed(execute=@n2.then_9, pawnVelocitySpeed:double=Vector Length(A=Get PawnVelocity())) -> (Output_Get:double)
   > n25 = Set pawnVelocitySpeedCutoff(pawnVelocitySpeedCutoff:double=SelectFloat(A=0, B=1, bPickA="float > float"(A=@n24, B=1200) {is=title K2Node_CallFunction})) -> (Output_Get:double)
   > Set SpringAlphaSpeedCap(SpringAlphaSpeedCap:double=float * float!(A=@n25, B=Get SpringAlpha() -> (SpringAlpha:double)) -> (ReturnValue)) -> (Output_Get:double)
   Event OnParachuteOpened()
   > n26 = Set Is Sky Diving(Is Sky Diving:double=0) -> (Output_Get:double)
   n27 = Switch on EGliderType(execute=[@n29.then, @n26.then], Selection=Get PlayerGliderType() -> (PlayerGliderType=HangGlider))
   Set IsWingGlider(execute=[@n27.Cape, @n27.HangGlider, @n27.Stairs, @n27.Surfing, @n27.SurfingSimple, @n27.Umbrella], IsWingGlider:double=0) -> (Output_Get:double)
   Set IsWingGlider(execute=@n27.NoGlider, IsWingGlider=1)
   Event OnParachuteClosed()
   > Set IsWingGlider(IsWingGlider=0)
   Branch(Condition=Get bIsInFrontEnd())
   n28 = Get isInFrontendAndNotPlayingEmote() -> (isInFrontendAndNotPlayingEmote:bool)
   > Set ControlRigAlpha(ControlRigAlpha="To Float (Boolean)"(InBool=NOT Boolean(A=@n28)))
   n29 = Branch(Condition=Get bIsInFrontEnd())
   Set IsWingGlider(execute=@n29.else, IsWingGlider=0)
   Set CharacterAimYaw_Sprint(CharacterAimYaw_Sprint:double=0) -> (Output_Get:double)
   > Set CharacterAimYaw_Idle(CharacterAimYaw_Idle:double=0) -> (Output_Get:double)
   PrintString!(execute, InString="To String (Boolean)"(InBool=Get bIsJumping()), bPrintToScreen, bPrintToLog, TextColor, Duration, Key) -> (then)
   PrintString(InString="To String (Float)"(InDouble=Get PawnVelocityZ()))
   GetOwningComponent()
   "To String (Boolean)"(InBool=Get bIsOnGround())
   Event BlueprintPostEvaluateAnimation() {is=title K2Node_Event}
   Event BlueprintInitializeAnimation()
   Get bIsFalling()
   Get bIsOnGround()
   Get bIsSurfaceSwimming()
   Get bIsMoving2D()
   Get bIsSprinting()
   Get Player Pawn Athena()
   Get bIsFalling()
   Get PawnSpeed2D()
   Get CustomEmoteNotPlaying() -> (CustomEmoteNotPlaying:double)
   Get bIsJumping()
  ]}
 Anim {
  path=EventGraph.K2Node_Composite_0.Anim
  nodes [
   Inputs()
   > n1 = Sequence!(execute) -> (then_0, then_1)
   > n2 = Branch(Condition=Get bIsInFrontEnd())
   n3 = Branch(execute=@n2.else, Condition=IsNullItem(ItemObject=Get Player AnimBP()))
   n4 = Get Player Pawn Athena()
   > n5 = Cast To Fortnite_M_Avg_Player_AnimBlueprint!(execute, Object=GetAnimInstance(self=Get Mesh(self=@n4))) -> (then, CastFailed, AsFortnite M Avg Player Anim Blueprint:Fortnite M Avg Player Anim Blueprint Object Reference) {is=title K2Node_DynamicCast}
   > n6 = Set Player AnimBP(Player AnimBP:Fortnite M Avg Player Anim Blueprint Object Reference=@n5) -> (Output_Get:Fortnite M Avg Player Anim Blueprint Object Reference)
   Outputs!(OutputPin:Exec=[@n6.then, @n3.else])
   Set BypassControlRigAlpha(execute=@n1.then_1, BypassControlRigAlpha:float=GetCurveValue(CurveName=BypassControlRig)) -> (Output_Get:float)
   n7 = Get Player Pawn Athena()
   n8 = GetBaseAimRotation!(self:Pawn Object Reference=@n7) -> (ReturnValue:Rotator)
   Set CharacterAimPitch(CharacterAimPitch:float=float / float(A=BreakRotator(InRot=@n8).Pitch, B=2)) -> (Output_Get:float)
   n9 = Get Mesh(self=@n7)
   n10 = BreakVector(InVec=InverseTransformDirection!(
     T:"Transform (by ref)"=GetSocketTransform(self=@n9, InSocketName=root, TransformSpace=RTS_World)
     Direction:Vector=GetForwardVector!(InRot:Rotator=@n8) -> (ReturnValue:Vector)
    ) -> (ReturnValue:Vector))
   > n11 = Set CharacterAimYaw_Sprint(CharacterAimYaw_Sprint=float * float(A="Atan2 (Degrees)"!(Y:double=@n10, X:double=@n10.Y) -> (ReturnValue:double), B=-1))
   > Set CharacterAimYaw_Idle(CharacterAimYaw_Idle=float / float(A=@n11, B=2))
  ]
 }
]
variables [
 BypassControlRigAlpha {type=number}
 CharacterAimPitch {type=number}
 CharacterAimYaw_Idle {type=number}
 CharacterAimYaw_Sprint {type=number}
 ControlRigAlpha {type=number}
 CustomEmoteBool {type=boolean}
 CustomEmoteNotPlaying {type=number}
 easingRunning {type=number}
 FrontEndCustomEmote {type=number}
 frontEndNoEmote {type=boolean}
 FrontEndNoEmoteExceptCustomEmote {type=boolean}
 Gravity Override Params {type=GravityOverrideParamsStruct}
 GravityOverride {type=Vector}
 Is Sky Diving {type=number}
 IsCustomEmotePlaying {type=number}
 isInFrontendAndNotPlayingEmote {type=boolean}
 IsWingGlider {type=number}
 NotJumping {type=number}
 notSkydiveSpringSetting {type=number}
 pawnVelocitySpeed {type=number}
 pawnVelocitySpeedCutoff {type=number}
 Player AnimBP {type=Fortnite_M_Avg_Player_AnimBlueprint_C, default=null}
 Player Pawn Athena {type=PlayerPawn_Athena_C}
 SpringAlpha {type=number}
 SpringAlphaSpeedCap {type=number}
 SpringStrength {type=number}
 SurfaceSwimAmount {type=number}
]
